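; Gameplay ability for the CoffeePuff AI: a blink punch that teleports it near its target.
; It takes over behavior control, picks a spot via EQS, and spawns a target-tracking visualization there.
; After an intro montage, punch-montage gameplay tags stop tracking, teleport the pawn with trail and portal cues, and start hit checks.
; Targeting repeats on a timer until slide end, damaging each actor once and sometimes triggering gloat speech.
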(event Ability|EventActivateAbility
  (CallFunction|InitAbility)
  (Ability|CommitAbility)
  (bind _returnvalue (AI|EQS|RunEQSQuery "/VividRazor_CoffeePuff/Abilities/EQS/EQS_FindPunchTeleportPosition_CoffeePuff.EQS_FindPunchTeleportPosition_CoffeePuff" (Variables|Default|GetAIPawn) "RandomBest5Pct"))
  (AI|EQS|BindEventtoOnQueryFinished _returnvalue (delegate Custom|OnEQSQueryFinished)))

(event Custom|OnEQSQueryFinished
  (params (QueryInstance :type EQS%20Query%20Instance%20Object%20Reference) (QueryStatus :type EEnvQueryStatus%20Enum :default "Processing"))
  (bind _teleportdestination (Variables|Default|GetTeleportDestination))
  (bind _beforeteleportlocation (Variables|Default|GetBeforeTeleportLocation))
  (if (Utilities|Enum|Equal(Enum) QueryStatus "Success")
    (CallFunction|SelectTeleportPoints QueryInstance)
    (CallFunction|SpawnVisualization)
    (Class|BPPlayerPawnNonParticipantVividRazorCoffeePuff|StartLookingAt (Variables|Default|GetAIPawn) (Variables|Default|GetTeleportLookAtLocation))
    (Ability|Tasks|PlayMontageandWaitforTrigger "None" "/VividRazor_CoffeePuff/Assets/Animations/Proto/TEMP_CoffeePuff_BlinkIntro_M.TEMP_CoffeePuff_BlinkIntro_M" 1.5
      (:Completed
        (Class|BPPlayerPawnNonParticipantVividRazorCoffeePuff|StopLookingAt (Variables|Default|GetAIPawn))
        (bind (_asynctaskproxy _targetdata _applicationtag) (Ability|Tasks|PlayMontageandWaitforTrigger "None" "/VividRazor_CoffeePuff/Assets/Animations/Proto/CoffeePuff_HFPunch.CoffeePuff_HFPunch" 0.4 "FirstSection" "None" (Variables|Default|GetAnimRootMotionScale))
          (:Triggered
            (switch Utilities|FlowControl|Switch|SwitchonGameplayTag _applicationtag
              (:Default
                (Ability|ApplyGameplayEffectContainer _targetdata _applicationtag))
              (:Event.Ability.VividRazor.CoffeePuff.BlinkWindupEnd
                (CallFunction|TriggerWindupVFX))
              (:Event.Ability.VividRazor.CoffeePuff.BlinkStopTracking
                (Variables|Default|SetLastTrackedTargetLocation (Transformation|GetActorLocation (Variables|Default|GetTargetPawn)))
                (Class|BGABlinkVisualization|SetTrackingActive (Variables|Default|GetVisualizationActor) false :!class /VividRazor_CoffeePuff/GameplayActors/BGA_Blink_Visualization.BGA_Blink_Visualization_C))
              (:Event.Ability.VividRazor.CoffeePuff.BlinkActivation
                (Variables|Default|SetBeforeTeleportLocation (Transformation|GetWorldLocation (Variables|/Script/Engine.Character|GetCapsuleComponent (Variables|Default|GetAIPawn))))
                (CallFunction|ExecuteTeleport)
                (CallFunction|AttemptHit)
                (Ability|ExecuteGameplayCueWithParamsOnOwner (Variables|Default|GetSpeedTrailEffectTag) (Ability|GameplayCue|MakeGameplayCueParameters :NormalizedMagnitude 1.0 :RawMagnitude 1.0 :Location _teleportdestination :Normal (Math|Rotator|GetRotationXVector (Math|Rotator|FindLookatRotation _beforeteleportlocation _teleportdestination)) :GameplayEffectLevel 1 :AbilityLevel 1 :bReplicateLocationWhenUsingMinimalRepProxy false))
                (Variables|Default|SetTeleportFXRotation (Math|Rotator|FindLookatRotation _teleportdestination _beforeteleportlocation))
                (Actor|DestroyActor (Variables|Default|GetVisualizationActor))
                (Utilities|FlowControl|Delay 0.1)
                (Ability|ExecuteGameplayCueWithParamsOnOwner (Variables|Default|GetExitPortalEffectTag) (Ability|GameplayCue|MakeGameplayCueParameters :NormalizedMagnitude 1.0 :RawMagnitude 1.0 :Location "0,0,0" :Normal (Math|Rotator|GetRotationXVector (Variables|Default|GetTeleportFXRotation)) :GameplayEffectLevel 1 :AbilityLevel 1 :bReplicateLocationWhenUsingMinimalRepProxy false)))
              (:Event.Ability.VividRazor.CoffeePuff.BlinkEndAbility
                _)
              (:Event.Ability.VividRazor.CoffeePuff.BlinkSlideEnd
                (Variables|Default|SetSlideCompleted true))))
          (:Cancelled
            (Development|PrintString "[BLINK Ability] Animation montage cancelled!" (Variables|Default|GetDebugOutput) (Variables|Default|GetDebugOutput))
            (Ability|EndAbility))
          (:Completed
            _))
        (Ability|EndAbility)))))

(event Ability|EventOnEndAbility
  (params (bWasCancelled :type Boolean))
  (bind _visualizationactor (Variables|Default|GetVisualizationActor))
  (AI|PopBehaviorControl (Variables|Default|GetAIBotController))
  (if (Utilities|IsValid _visualizationactor :!class /Script/Engine.KismetSystemLibrary)
    (Actor|DestroyActor _visualizationactor)))

(event Custom|AttemptHit
  (bind (_asynctaskproxy _targetinghandle) (Class|AsyncActionPerformTargeting|PerformTargetingAsyncAction (Variables|Default|GetAIPawn) (Variables|Default|GetTargetingPreset))
    (:Targeted
      (Utilities|Array|Clear (Variables|Default|GetNewTargets))
      (for _array_element (TargetingSystem|TargetingResults|GetTargetingResults (GameInstanceSubsystems|GetTargetingSubsystem) _targetinghandle)
        (if (not (Utilities|Array|ContainsItem (Variables|Default|GetAlreadyAffectedTargets) (.HitActor (Collision|BreakHitResult _array_element))))
          (Utilities|Array|Add (Variables|Default|GetNewTargets) _array_element)
          (Utilities|Array|Add (Variables|Default|GetAlreadyAffectedTargets) (.HitActor (Collision|BreakHitResult _array_element)))))
      (Ability|Tasks|ApplyGameplayEffectToHitResultsAsync (Ability|GetActivatingAbilityComponent) (Variables|Default|GetNewTargets) (Variables|Default|GetDamageEffectClass) 0 1 (Class|AthenaFunctionLibrary|MakeSimpleFortEffectContext (Variables|Default|GetAIPawn) (Variables|Default|GetAIPawn)))
      (CallFunction|TryGloating)
      (if (not (Variables|Default|GetSlideCompleted))
        (Utilities|Time|SetTimerbyEvent (EventDispatchers|CreateEvent :!event Custom|AttemptHit) (Variables|Default|GetHitRate))))))

(fn ExecuteTeleport
  (bind _aipawn (Variables|Default|GetAIPawn))
  (bind _capsulecomponent (Variables|/Script/Engine.Character|GetCapsuleComponent _aipawn))
  (bind _returnvalue (Components|Capsule|GetScaledCapsuleHalfHeight _capsulecomponent))
  (bind _returnvalue_2 (Components|Capsule|GetScaledCapsuleRadius _capsulecomponent))
  (bind _returnvalue_3 (- _returnvalue_2))
  (bind _returnvalue_4 (Math|Vector|MakeVector _returnvalue_3 0.0 _returnvalue))
  (bind _returnvalue_5 (Transformation|GetActorForwardVector _aipawn))
  (bind _teleporttargetlocation (Variables|Default|GetTeleportTargetLocation))
  (bind _returnvalue_6 (Utilities|Operators|Add _teleporttargetlocation _returnvalue_4 (* _returnvalue_5 (* (Variables|Default|GetAnimRootMotionScale) -2e+02))))
  (bind _returnvalue_7 (Utilities|String|ToString(Vector) _returnvalue_6))
  (bind (_x _y _z) (Math|Rotator|BreakRotIntoAxes (Transformation|GetActorRotation _aipawn)))
  (bind _animrootmotionscale (Variables|Default|GetAnimRootMotionScale))
  (bind _returnvalue_8 (* _animrootmotionscale -2e+02))
  (bind _returnvalue_9 (* _returnvalue_5 _returnvalue_8))
  (bind _debugoutput (Variables|Default|GetDebugOutput))
  (Variables|Default|SetTeleportDestination _returnvalue_6)
  (bind _returnvalue_10 (Transformation|Teleport _aipawn _returnvalue_6 (Math|Rotator|MakeRotationfromAxes (- (Variables|Default|GetLastTrackedTargetLocation) (Variables|Default|GetTeleportTargetLocation)) _y _z)))
  (if _returnvalue_10
    (Development|PrintString (Utilities|String|Append "[BLINK Ability] Succesfully teleported to (" _returnvalue_7 ").") _debugoutput)
    (else
      (Development|PrintString (Utilities|String|Append "[BLINK Ability] FAILED teleporting to (" _returnvalue_7 ").") _debugoutput))))

(fn InitAbility
  (bind _asbp_player_pawn_non_participant_vivid_razor_coffee_puff (Utilities|Casting|CastToBP_PlayerPawn_NonParticipant_VividRazor_CoffeePuff (CallFunction|GetAvatarasFortPlayerPawn)))
  (bind _output_get (Variables|Default|SetAIPawn _asbp_player_pawn_non_participant_vivid_razor_coffee_puff))
  (bind _asfort_athena_aibot_controller (Utilities|Casting|CastToFortAthenaAIBotController (Pawn|GetController _output_get)))
  (Variables|Default|SetAIBotController _asfort_athena_aibot_controller)
  (bind _targetactor (Class|BPFLAICommonFunctions|GetTargetActorfromBlackboard (Variables|Default|GetAIPawn)))
  (Variables|Default|SetTargetPawn _targetactor)
  (AI|PushBehaviorControl (Variables|Default|GetAIBotController) -2097152)
  (Class|FortAthenaAIBotController|StopFire (Variables|Default|GetAIBotController))
  (Variables|Default|SetSlideCompleted false)
  (Utilities|Array|Clear (Variables|Default|GetNewTargets))
  (Utilities|Array|Clear (Variables|Default|GetAlreadyAffectedTargets)))

(fn CanActivateAbility
  (params (ActorInfo :type Gameplay%20Ability%20Actor%20Info%20Structure) (Handle :type Gameplay%20Ability%20Spec%20Handle%20Structure))
  (return 0 true))

(fn SelectTeleportPoints
  (params (EnvQueryResults :type EQS%20Query%20Instance%20Object%20Reference))
  (bind (_returnvalue_x _returnvalue_y _returnvalue_z) (Math|Vector|BreakVector (Pawn|GetNavAgentLocation (Variables|Default|GetAIPawn))))
  (bind _targetpawn (Variables|Default|GetTargetPawn))
  (bind _returnvalue (Pawn|GetNavAgentLocation _targetpawn))
  (bind _teleporttargetlocation (Variables|Default|GetTeleportTargetLocation))
  (bind _resultlocations (AI|EQS|GetQueryResultsasLocations EnvQueryResults))
  (for (_array_element _array_index) _resultlocations
    _)
  (bind _output_get (Variables|Default|SetTeleportTargetLocation (Utilities|Array|Get(acopy) _resultlocations)))
  (bind _returnvalue_2 (Fortnite|GetAngleDegrees (- (Pawn|GetNavAgentLocation (Variables|Default|GetAIPawn)) _returnvalue) (- _teleporttargetlocation _returnvalue)))
  (bind _output_get_2 (Variables|Default|SetTeleportLookAtLocation (Math|Vector|SelectVector (Math|Vector|MakeVector (+ _returnvalue_x (* (Math|Vector|Distance(Vector) _teleporttargetlocation _returnvalue) (Math|Trig|Cos(Radians) (Math|Trig|DegreesToRadians _returnvalue_2)))) (+ _returnvalue_y (* (Math|Vector|Distance(Vector) _teleporttargetlocation _returnvalue) (Math|Trig|Sin(Radians) (Math|Trig|DegreesToRadians _returnvalue_2)))) _returnvalue_z) (Math|Vector|MakeVector 0.0 0.0 (Math|Vector|Distance(Vector) (Pawn|GetNavAgentLocation (Variables|Default|GetAIPawn)) (Math|Vector|MakeVector (+ _returnvalue_x (* (Math|Vector|Distance(Vector) _teleporttargetlocation _returnvalue) (Math|Trig|Cos(Radians) (Math|Trig|DegreesToRadians _returnvalue_2)))) (+ _returnvalue_y (* (Math|Vector|Distance(Vector) _teleporttargetlocation _returnvalue) (Math|Trig|Sin(Radians) (Math|Trig|DegreesToRadians _returnvalue_2)))) _returnvalue_z))) true))))

(fn SpawnVisualization
  (bind _returnvalue (Building|SpawnBuildingGameplayActor "/VividRazor_CoffeePuff/GameplayActors/BGA_Blink_Visualization.BGA_Blink_Visualization_C" (Math|Transform|MakeTransform (+ (Variables|Default|GetTeleportTargetLocation) (Math|Vector|MakeVector 0.0 0.0 (Components|Capsule|GetScaledCapsuleHalfHeight (Variables|/Script/Engine.Character|GetCapsuleComponent (Variables|Default|GetAIPawn)))))) (Variables|Default|GetAIPawn) true false))
  (bind _asbga_blink_visualization (Utilities|Casting|CastToBGA_Blink_Visualization _returnvalue))
  (bind _output_get (Variables|Default|SetVisualizationActor _asbga_blink_visualization))
  (Class|BGABlinkVisualization|SetTargetPawn _output_get (Variables|Default|GetTargetPawn) :!class /VividRazor_CoffeePuff/GameplayActors/BGA_Blink_Visualization.BGA_Blink_Visualization_C))

(fn TriggerWindupVFX
  (GameplayCue|ExecuteGameplayCue (Variables|Default|GetAIPawn) (Variables|Default|GetBlinkTrailGameplayCueTag)))

(fn TryGloating
  (if (> (Utilities|Array|Length (Variables|Default|GetNewTargets)) 0)
    (bind _returnvalue (GameplayTags|HasMatchingGameplayTag(Message) (Variables|Default|GetAIPawn) "(TagName=\"Athena.AI.SpeechBubbles.VividRazor.CoffeePuff.MidCombatSpeechCooldown\")"))
    (if (not _returnvalue)
      (if (< (Math|Random|RandomFloat) 0.5)
        (Class|BPFLAICommonFunctions|PlaySpeechBubble (Variables|Default|GetAIPawn) (Variables|Default|GetGloatSpeechTag) true true)
        (Ability|ApplyGameplayEffectToOwner (Variables|Default|GetGloatCooldownTag))))))
