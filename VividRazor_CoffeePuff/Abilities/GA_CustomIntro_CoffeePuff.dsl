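; GA_CustomIntro_CoffeePuff is a gameplay ability that plays a custom intro for the CoffeePuff AI character.
; After a short delay it gets the avatar's Fort pawn, ending immediately if the cast fails, then plays the configured montage.
; Shortly after the montage starts it shows the combat-start speech bubble, with a DoOnce guarding EndAbility.
; On end, it adds the ability-end loose gameplay tag to the pawn if valid.
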(event Ability|EventActivateAbility
  (bind (_montageinfo_montagetoplay _montageinfo_animplayrate _montageinfo_animrootmotiontranslationscale _montageinfo_montagesectiontoplay _montageinfo_overridesection) (Utilities|Struct|BreakFortGameplayAbilityMontageInfo (Variables|GetMontageInfo_MontageToPlay)))
  (Utilities|FlowControl|Delay 0.8)
  (bind _outfortplayerpawn (CallFunction|GetAvatarasValidFortPlayerPawn)
    (:ValidFortPlayerPawn
      (Variables|Default|SetAIPawn _outfortplayerpawn)
      (Ability|Tasks|PlayMontageAndWaitNotify "None" _montageinfo_montagetoplay _montageinfo_animplayrate _montageinfo_overridesection true _montageinfo_animrootmotiontranslationscale
        (:then
          (Utilities|FlowControl|Delay 0.154)
          (Class|BPFLAICommonFunctions|PlaySpeechBubble (Variables|Default|GetAIPawn) (Variables|Default|GetCombatStartSpeechTag) true true)
          (Utilities|FlowControl|Delay 4.0))
        (:OnCompleted
          _)
        (:OnInterrupted
          _)
        (:OnCancelled
          _))
      (Utilities|FlowControl|DoOnce false
        (:Completed
          (Ability|EndAbility))))
    (:AvatarCastFailed
      (Ability|EndAbility))))

(event Ability|EventOnEndAbility
  (params (bWasCancelled :type Boolean))
  (bind _aipawn (Variables|Default|GetAIPawn))
  (Utilities|IsValid _aipawn
    (:Is%20Valid
      (Ability|GameplayEffect|AddLooseGameplayTags _aipawn (Variables|Default|GetAbilityEndTag) true))))
