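; CoffeePuff AI boss ability that plays a montage to summon and detonate a fire pillar at its blackboard target.
; Once playlist data loads, it takes over bot behavior, stops firing, spends energy scaled by squad size, then commits.
; Montage notify tags spawn the pillar at the target's navmesh-projected location, then trigger its burst.
; Activation is blocked during knockback or falling, and ending the ability releases behavior control.
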(event Ability|EventActivateAbility
  (bind (_gamestate _playlist) (Class|FortAsyncActionCurrentPlaylistReady|PlaylistDataReadyAsync)
    (:OnReady
      (Variables|Default|SetMaxTeamSize (Variables|/Script/FortniteGame.FortPlaylist|GetMaxSquadSize _playlist))
      (Development|PrintString "[GA_FirePillar_CoffeePuff]: Activating ability!" (Variables|Default|GetDebugOutput) (Variables|Default|GetDebugOutput) (Variables|Default|GetDebugOutputColor) (Variables|Default|GetDebugOutputDuration))
      (CallFunction|InitAbility)
      (bind _returnvalue (Ability|CommitAbility))
      (if _returnvalue
        (CallFunction|InitiateFirePillar)
        (else
          (Ability|EndAbility))))))

(event Custom|InitiateFirePillar
  (bind (_projectedlocation _returnvalue) (AI|Navigation|ProjectPointtoNavigation (Pawn|GetNavAgentLocation (Variables|Default|GetTargetPawn)) (Fortnite|AI|GetNavigationDataforActor (Variables|Default|GetAIPawn)) "/Game/Athena/AI/Phoebe/NavFilter/NavFilter_NPC_Base.NavFilter_NPC_Base_C" "70.000000,70.000000,70.000000"))
  (bind (_montageinfo_montagetoplay _montageinfo_animplayrate _montageinfo_animrootmotiontranslationscale _montageinfo_montagesectiontoplay _montageinfo_overridesection) (Utilities|Struct|BreakFortGameplayAbilityMontageInfo (Variables|GetMontageInfo_MontageToPlay)))
  (bind (_asynctaskproxy _notifytag) (Ability|Tasks|PlayMontageAndWaitNotify "CoffeePuff_FirePillar" _montageinfo_montagetoplay _montageinfo_animplayrate _montageinfo_overridesection true _montageinfo_animrootmotiontranslationscale)
    (:OnCompleted
      _)
    (:OnInterrupted
      _)
    (:OnCancelled
      _)
    (:OnNotifyBegin
      (switch Utilities|FlowControl|Switch|SwitchonGameplayTag _notifytag
        (:Event.Ability.VividRazor.CoffeePuff.FirePillarActivation
          (bind _returnvalue_2 (Building|SpawnBuildingGameplayActor "/VividRazor_CoffeePuff/GameplayActors/BGA_CoffeePuff_FirePillar.BGA_CoffeePuff_FirePillar_C" (Math|Transform|MakeTransform (select _returnvalue _projectedlocation (Pawn|GetNavAgentLocation (Variables|Default|GetTargetPawn)))) (Variables|Default|GetAIPawn) true false))
          (bind _asbga_coffee_puff_fire_pillar (Utilities|Casting|CastToBGA_CoffeePuff_FirePillar _returnvalue_2))
          (Variables|Default|SetFirePillar _asbga_coffee_puff_fire_pillar))
        (:Event.Ability.VividRazor.CoffeePuff.FirePillarBurst
          (Class|BGACoffeePuffFirePillar|Detonate (Variables|Default|GetFirePillar))))))
  (Ability|EndAbility))

(event Ability|EventOnEndAbility
  (params (bWasCancelled :type Boolean))
  (bind _aibotcontroller (Variables|Default|GetAIBotController))
  (Development|PrintString "[GA_FirePillar_CoffeePuff]: Ending ability!" (Variables|Default|GetDebugOutput) (Variables|Default|GetDebugOutput) (Variables|Default|GetDebugOutputColor) (Variables|Default|GetDebugOutputDuration))
  (Utilities|IsValid _aibotcontroller
    (:Is%20Valid
      (AI|PopBehaviorControl _aibotcontroller))))

(fn InitAbility
  (bind _maxteamsize (Variables|Default|GetMaxTeamSize))
  (bind _returnvalue (Math|Conversions|ToFloat(Integer) _maxteamsize))
  (bind _aipawn (Variables|Default|GetAIPawn))
  (bind _returnvalue_2 (Actor|GetComponentbyClass _aipawn "/Script/EnergyRuntime.FortComponent_Energy"))
  (bind _energycost (Variables|Default|GetEnergyCost))
  (bind _returnvalue_3 (ScalableFloat|GetValueatLevel _energycost _returnvalue))
  (bind _output_get (Variables|Default|SetAIPawn (CallFunction|GetAvatarasFortPlayerPawn)))
  (bind _targetactor (Class|BPFLAICommonFunctions|GetTargetActorfromBlackboard _output_get))
  (bind _output_get_2 (Variables|Default|SetTargetPawn _targetactor))
  (bind _asbp_player_pawn_non_participant_vivid_razor_coffee_puff (Utilities|Casting|CastToBP_PlayerPawn_NonParticipant_VividRazor_CoffeePuff (Variables|Default|GetAIPawn))
    (:then
      (Class|BPPlayerPawnNonParticipantVividRazorCoffeePuff|UpdateLastKnownValidTargetLocation _asbp_player_pawn_non_participant_vivid_razor_coffee_puff (Pawn|GetNavAgentLocation _output_get_2)))
    (:CastFailed
      _))
  (bind _asfort_athena_aibot_controller (Utilities|Casting|CastToFortAthenaAIBotController (Pawn|GetController (Variables|Default|GetAIPawn)))
    (:then
      (bind _output_get_3 (Variables|Default|SetAIBotController _asfort_athena_aibot_controller))
      (AI|PushBehaviorControl _output_get_3 -2097152)
      (Class|FortAthenaAIBotController|StopFire (Variables|Default|GetAIBotController))
      (Energy|UseEnergy _returnvalue_2 _returnvalue_3 true 0)
      (Development|PrintString (Utilities|String|Append "[GA_FirePillar_CoffeePuff] Spending " (Utilities|String|ToString(Float) _returnvalue_3) " energy! Current energy: " (Utilities|String|ToString(Float) (Energy|GetCurrentEnergyPercentage _returnvalue_2))) (Variables|Default|GetDebugOutput) (Variables|Default|GetDebugOutput) (Variables|Default|GetDebugOutputColor) (Variables|Default|GetDebugOutputDuration)))
    (:CastFailed
      (Ability|EndAbility))))

(fn CanActivateAbility
  (params (ActorInfo :type Gameplay%20Ability%20Actor%20Info%20Structure) (Handle :type Gameplay%20Ability%20Spec%20Handle%20Structure))
  (bind _asfort_pawn (Utilities|Casting|CastToFortPawn (Utilities|Struct|BreakGameplayAbilityActorInfo ActorInfo)))
  (return 0 (and (not (GameplayTags|HasAnyMatchingGameplayTags _asfort_pawn (GameplayTags|MakeGameplayTagContainerfromArray (Utilities|Array|MakeArray "(TagName=\"Gameplay.Status.Knockback\")")))) (not (AI|Components|NavMovement|IsFalling (Variables|/Script/Engine.Character|GetCharacterMovement _asfort_pawn))))))
