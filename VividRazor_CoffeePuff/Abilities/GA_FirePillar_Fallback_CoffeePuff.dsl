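; Gameplay ability for CoffeePuff's fallback fire pillar attack, adding only an activation gate, with empty activate and end events.
; CanActivateAbility requires the parent check to pass and the blackboard target to be moving on ground.
; When a scalable-float toggle is on, the target's nav agent location must also fail projection onto the navmesh, using the Phoebe nav filter and tunable probe extents, so it targets off-navmesh players.
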(event Ability|EventActivateAbility
  (|CallParentFunction) ; This node is disabled and will not be called.\nDrag off pins to build functionality.
)

(event Ability|EventOnEndAbility
  (params (bWasCancelled :type Boolean))
  (|CallParentFunction bWasCancelled) ; This node is disabled and will not be called.\nDrag off pins to build functionality.
)

(fn CanActivateAbility
  (params (ActorInfo :type Gameplay%20Ability%20Actor%20Info%20Structure) (Handle :type Gameplay%20Ability%20Spec%20Handle%20Structure))
  (bind _targetactor (Class|BPFLAICommonFunctions|GetTargetActorfromBlackboard (CallFunction|GetAvatarasFortPlayerPawn)))
  (bind (_relevanttags _returnvalue) (|CallParentFunction ActorInfo Handle))
  (return _relevanttags (Math|Boolean|ANDBoolean _returnvalue (AI|Components|NavMovement|IsMovingonGround (Variables|/Script/Engine.Character|GetCharacterMovement _targetactor)) (select (> (ScalableFloat|GetValueatLevel (LocalVariables|GetTargetHasToBeOffNavmesh)) 0) (not (.ReturnValue (AI|Navigation|ProjectPointtoNavigation (Pawn|GetNavAgentLocation _targetactor) (Fortnite|AI|GetNavigationDataforActor _targetactor) "/Game/Athena/AI/Phoebe/NavFilter/NavFilter_Phoebe.NavFilter_Phoebe_C" (Math|Vector|MakeVector (ScalableFloat|GetValueatLevel (LocalVariables|GetTargetNavProbe_ExtentX)) (ScalableFloat|GetValueatLevel (LocalVariables|GetTargetNavProbe_ExtentY)) (ScalableFloat|GetValueatLevel (LocalVariables|GetTargetNavProbe_ExtentZ)))))) true))))
