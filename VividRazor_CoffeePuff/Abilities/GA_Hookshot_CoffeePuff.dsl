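; Gameplay ability for the CoffeePuff boss AI that throws a hookshot to damage and pull its target.
; It spends energy, plays a montage and locks aim after a delay, with tuning indexed by playlist squad size.
; The Throw notify spawns a homing projectile whose turn speed can scale with aim mismatch.
; Pawn hits apply effects and jump to PullSuccess; other hits or timeouts jump to PullFail, and ending restores AI focus.
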(event Ability|EventActivateAbility
  (bind (_montageinfo_montagetoplay _montageinfo_animplayrate _montageinfo_animrootmotiontranslationscale _montageinfo_montagesectiontoplay _montageinfo_overridesection) (Utilities|Struct|BreakFortGameplayAbilityMontageInfo (Variables|GetMontageInfo_MontageToPlay)))
  (bind (_gamestate _playlist) (Class|FortAsyncActionCurrentPlaylistReady|PlaylistDataReadyAsync)
    (:OnReady
      (Variables|Default|SetMaxTeamSize (Variables|/Script/FortniteGame.FortPlaylist|GetMaxSquadSize _playlist))
      (CallFunction|InitAbility)
      (Ability|CommitAbility)
      (bind (_asynctaskproxy _notifytag) (Ability|Tasks|PlayMontageAndWaitNotify "Task_PlayMontage" _montageinfo_montagetoplay _montageinfo_animplayrate _montageinfo_overridesection false _montageinfo_animrootmotiontranslationscale)
        (:then
          (if (Variables|Debug|GetDebugOutput)
            ;[0.01077 0.15 0 0.5] Debug Draw for Aiming
            (Variables|Debug|SetAimDebugStartTimecode (Utilities|Time|GetTimeSeconds))
            (if (and (< (- (Utilities|Time|GetTimeSeconds) (Variables|Debug|GetAimDebugStartTimecode)) (ScalableFloat|GetValueatLevel (Variables|Aiming|GetLockFocusDelay) (Math|Conversions|ToFloat(Integer) (Variables|Default|GetMaxTeamSize)))) (not (Utilities|IsValid (Variables|Projectile|GetHookshotProjectile) :!class /Script/Engine.KismetSystemLibrary)))
              (CallFunction|DebugDrawAimPosition 2e+01 1.0 "(R=0.000000,G=1.000000,B=0.607843,A=1.000000)")
              (Utilities|FlowControl|DelayUntilNextTick))
            (CallFunction|SetLockedFocusPosition)
            (CallFunction|DebugDrawAimPosition 22.0 1.5 "(R=0.035294,G=0.000000,B=1.000000,A=1.000000)" true)
            (CallFunction|DebugDrawAimPosition 2e+01 1.0 "(R=0.000000,G=1.000000,B=0.607843,A=1.000000)")
            (Utilities|FlowControl|DelayUntilNextTick)
            (else
              (Utilities|FlowControl|Delay (ScalableFloat|GetValueatLevel (Variables|Aiming|GetLockFocusDelay) (Math|Conversions|ToFloat(Integer) (Variables|Default|GetMaxTeamSize))))
              (CallFunction|SetLockedFocusPosition))))
        (:OnCompleted
          _)
        (:OnInterrupted
          _)
        (:OnCancelled
          _)
        (:OnNotifyBegin
          (switch Utilities|FlowControl|Switch|SwitchonGameplayTag _notifytag
            (:Event.Ability.VividRazor.CoffeePuff.Hookshot.Throw
              (if (not (Variables|Aiming|GetAimLocked))
                (CallFunction|SetLockedFocusPosition))
              (Utilities|FlowControl|Sequence
                (:then_0
                  (Utilities|FlowControl|Delay (ScalableFloat|GetValueatLevel (Variables|Projectile|GetMaxTimeBeforePull) (Math|Conversions|ToFloat(Integer) (Variables|Default|GetMaxTeamSize))))
                  (if (not (Variables|Projectile|GetHookedSomething))
                    (Ability|Animation|MontageJumptoSection "PullFail")
                    (CallFunction|DisableProjectile)
                    (Development|PrintString "[GA_Hookshot_CoffeePuff] Maximum hook flight time reached, recalling hook!" (Variables|Debug|GetDebugOutput) (Variables|Debug|GetDebugOutput) (Variables|Debug|GetDebugOutputColor) (Variables|Debug|GetDebugOutputDuration))))
                (:then_1
                  (bind (_asynctaskproxy_2 _projectiledata) (Projectile|Tasks|SpawnProjectileAndWait :Class (Variables|Projectile|GetHookProjectileClass) :RequestedBy (Variables|Default|GetAIPawn) :SpawnLocation (CallFunction|GetProjectileSpawnLocation) :SpawnRotation (Transformation|GetActorRotation (Variables|Default|GetAIPawn)) :SpawnDirection (Math|Conversions|RotationFromXVector (- (Variables|Aiming|GetLockedFocusPosition) (CallFunction|GetProjectileSpawnLocation))) :InitialSpeed (ScalableFloat|GetValueatLevel (Variables|Projectile|GetProjectileInitialSpeed) (Math|Conversions|ToFloat(Integer) (Variables|Default|GetMaxTeamSize))) :GravityScale 0.0 :bAllowProjectileSpawnWhenDead false :bAllowProjectileSpawnWhenDBNO false)
                    (:then
                      (bind (_projectiledata_spawnedprojectile _projectiledata_hits) (Utilities|Struct|BreakProjectileEventData _projectiledata))
                      (Projectile|SetHomingData _projectiledata_spawnedprojectile _projectiledata_spawnedprojectile (Utilities|Struct|MakeProjectileHomingData "LockOn" (CallFunction|GetHomingTurnSpeedMin) (CallFunction|GetHomingTurnSpeedMax) 0.15 0.0 -1 (Variables|Default|GetTargetPawn) 0 "0.000000,0.000000,0.000000" (Math|Vector|MakeVector 0.0 0.0 (ScalableFloat|GetValueatLevel (Variables|Aiming|GetAimZOffset) (Math|Conversions|ToFloat(Integer) (Variables|Default|GetMaxTeamSize))))) :!class /Script/FortniteGame.FortProjectileMovementInterface)
                      (Ability|AddGameplayCueToOwner (Variables|Cues|GetCastProjectileTag)))
                    (:Created
                      (bind _asb_prj_hookshot (Utilities|Casting|CastToB_Prj_Hookshot _projectiledata_spawnedprojectile))
                      (bind _output_get (Variables|Projectile|SetHookshotProjectile _asb_prj_hookshot))
                      (bind _asbp_player_pawn_non_participant_vivid_razor_coffee_puff (Utilities|Casting|CastToBP_PlayerPawn_NonParticipant_VividRazor_CoffeePuff (Variables|Default|GetAIPawn)))
                      (Class|BPPlayerPawnNonParticipantVividRazorCoffeePuff|SetActiveHookshot _asbp_player_pawn_non_participant_vivid_razor_coffee_puff _output_get :!class /VividRazor_CoffeePuff/NPCs/CoffeePuff/Pawn/BP_PlayerPawn_NonParticipant_VividRazor_CoffeePuff.BP_PlayerPawn_NonParticipant_VividRazor_CoffeePuff_C))
                    (:Stopped
                      (bind _asfort_pawn (Utilities|Casting|CastToFortPawn (.HitActor (Collision|BreakHitResult (Utilities|Array|Get(aref) _projectiledata_hits))))
                        (:then
                          (Variables|Default|SetAs%20Fort%20Pawn _asfort_pawn)
                          (Variables|Projectile|SetHookedSomething true)
                          (Ability|ApplyGameplayEffectContainer (Ability|TargetData|AbilityTargetDatafromHitResult (Utilities|Array|Get(aref) _projectiledata_hits)) "(TagName=\"Ability.VividRazor.Boss.CoffeePuff.Hookshot.Damage\")")
                          (Ability|ApplyGameplayEffectContainer (Ability|TargetData|AbilityTargetDatafromHitResult (Utilities|Array|Get(aref) _projectiledata_hits)) "(TagName=\"Ability.VividRazor.Boss.CoffeePuff.Hookshot.Pull\")")
                          (CallFunction|DisableProjectile (Utilities|Array|Get(aref) _projectiledata_hits))
                          (GameplayCue|ExecuteGameplayCue (Variables|Default|GetAs%20Fort%20Pawn) (Variables|Cues|GetImpactPawnTag))
                          (Utilities|FlowControl|Delay (Utilities|Operators|Add (ScalableFloat|GetValueatLevel (Variables|Projectile|GetKnockbackPostDelay) (Math|Conversions|ToFloat(Integer) (Variables|Default|GetMaxTeamSize))) (ScalableFloat|GetValueatLevel (Variables|Projectile|GetKnockbackDuration) (Math|Conversions|ToFloat(Integer) (Variables|Default|GetMaxTeamSize))) 0.1))
                          (Ability|Animation|MontageJumptoSection "PullSuccess")
                          (CallFunction|PlayRecallSound))
                        (:CastFailed
                          (Ability|Animation|MontageJumptoSection "PullFail")
                          (CallFunction|DisableProjectile)
                          (GameplayCue|ExecuteGameplayCue (Variables|Default|GetAIPawn) (Variables|Cues|GetImpactWorldTag)))))
                    (:Destroyed
                      (CallFunction|DisableProjectile (Utilities|Array|Get(aref) _projectiledata_hits))))))))))
      (Ability|EndAbility))))

(event Ability|EventOnEndAbility
  (params (bWasCancelled :type Boolean))
  (bind _aibotcontroller (Variables|Default|GetAIBotController))
  (Development|PrintString "[GA_Hookshot_CoffeePuff]: Ending ability!" (Variables|Debug|GetDebugOutput) (Variables|Debug|GetDebugOutput) (Variables|Debug|GetDebugOutputColor) (Variables|Debug|GetDebugOutputDuration))
  (CallFunction|DisableProjectile)
  (Utilities|IsValid _aibotcontroller
    (:Is%20Valid
      (AI|PopBehaviorControl _aibotcontroller)
      (Class|FortAthenaAIBotController|ResumeTargetFocusing _aibotcontroller)
      (AI|SetFocus _aibotcontroller (Variables|Default|GetTargetPawn))
      (bind _asbp_player_pawn_non_participant_vivid_razor_coffee_puff (Utilities|Casting|CastToBP_PlayerPawn_NonParticipant_VividRazor_CoffeePuff (Variables|Default|GetAIPawn)))
      (Class|BPPlayerPawnNonParticipantVividRazorCoffeePuff|ClearActiveHookshot _asbp_player_pawn_non_participant_vivid_razor_coffee_puff))))

(fn DebugDrawAimPosition
  (params (Radius :type Float%20%28double-precision%29 :default 2e+01) (Thickness :type Float%20%28double-precision%29 :default 1.0) (Color :type Linear%20Color%20Structure) (DrawLockedPosition :type Boolean))
  (bind (_returnvalue_x _returnvalue_y _returnvalue_z) (Math|Vector|BreakVector (Transformation|GetActorLocation (Variables|Default|GetTargetPawn))))
  (Debug|DebugSphereSendtoAll (Math|Vector|SelectVector (Variables|Aiming|GetLockedFocusPosition) (Math|Vector|MakeVector _returnvalue_x _returnvalue_y (+ _returnvalue_z (ScalableFloat|GetValueatLevel (Variables|Aiming|GetAimZOffset) (Math|Conversions|ToFloat(Integer) (Variables|Default|GetMaxTeamSize))))) (Variables|GetDrawLockedPosition)) (Variables|GetRadius) 64 (Math|Conversions|ToColor(LinearColor) (Variables|GetColor)) false -1.0 0 (Variables|GetThickness)))

(fn DisableProjectile
  (params (HitResult :type Hit%20Result%20Structure))
  (Ability|HasAuthority
    (:True
      (Class|BPrjHookshot|DisableProjectile (Variables|Projectile|GetHookshotProjectile) HitResult))))

(fn GetAimMismatchAngle
  (bind (_returnvalue_x _returnvalue_y _returnvalue_z) (Math|Vector|BreakVector (Transformation|GetActorLocation (Variables|Default|GetTargetPawn))))
  (return (Math|Trig|Acos(Degrees) (Math|Vector|DotProduct (Math|Vector|Normalize (- (Math|Vector|MakeVector _returnvalue_x _returnvalue_y (+ _returnvalue_z (ScalableFloat|GetValueatLevel (Variables|Aiming|GetAimZOffset) (Math|Conversions|ToFloat(Integer) (Variables|Default|GetMaxTeamSize))))) (CallFunction|GetProjectileSpawnLocation))) (Math|Vector|Normalize (- (Variables|Aiming|GetLockedFocusPosition) (CallFunction|GetProjectileSpawnLocation)))))))

(fn GetHomingTurnSpeedMax
  (return (Math|Float|SelectFloat (ScalableFloat|GetValueatLevel (Variables|Projectile|GetProjectileMaxTurnSpeed) (Math|Conversions|ToFloat(Integer) (Variables|Default|GetMaxTeamSize))) (* (ScalableFloat|GetValueatLevel (Variables|Projectile|GetProjectileMaxTurnSpeed) (Math|Conversions|ToFloat(Integer) (Variables|Default|GetMaxTeamSize))) (* (/ (Math|Float|Clamp(Float) (- (CallFunction|GetAimMismatchAngle) (ScalableFloat|GetValueatLevel (Variables|Projectile|GetProjectileAimMismatchMin) (Math|Conversions|ToFloat(Integer) (Variables|Default|GetMaxTeamSize)))) 0.0 (- (ScalableFloat|GetValueatLevel (Variables|Projectile|GetProjectileAimMismatchMax) (Math|Conversions|ToFloat(Integer) (Variables|Default|GetMaxTeamSize))) (ScalableFloat|GetValueatLevel (Variables|Projectile|GetProjectileAimMismatchMin) (Math|Conversions|ToFloat(Integer) (Variables|Default|GetMaxTeamSize))))) (- (ScalableFloat|GetValueatLevel (Variables|Projectile|GetProjectileAimMismatchMax) (Math|Conversions|ToFloat(Integer) (Variables|Default|GetMaxTeamSize))) (ScalableFloat|GetValueatLevel (Variables|Projectile|GetProjectileAimMismatchMin) (Math|Conversions|ToFloat(Integer) (Variables|Default|GetMaxTeamSize))))) (ScalableFloat|GetValueatLevel (Variables|Projectile|GetProjectileAimMismatchMultiplier) (Math|Conversions|ToFloat(Integer) (Variables|Default|GetMaxTeamSize))))) (<= (ScalableFloat|GetValueatLevel (Variables|Projectile|GetProjectileTurnSpeedAffectedByAimMismatch) (Math|Conversions|ToFloat(Integer) (Variables|Default|GetMaxTeamSize))) 0))))

(fn GetHomingTurnSpeedMin
  (return (Math|Float|SelectFloat (ScalableFloat|GetValueatLevel (Variables|Projectile|GetProjectileMinTurnSpeed) (Math|Conversions|ToFloat(Integer) (Variables|Default|GetMaxTeamSize))) (* (ScalableFloat|GetValueatLevel (Variables|Projectile|GetProjectileMinTurnSpeed) (Math|Conversions|ToFloat(Integer) (Variables|Default|GetMaxTeamSize))) (* (/ (Math|Float|Clamp(Float) (- (CallFunction|GetAimMismatchAngle) (ScalableFloat|GetValueatLevel (Variables|Projectile|GetProjectileAimMismatchMin) (Math|Conversions|ToFloat(Integer) (Variables|Default|GetMaxTeamSize)))) 0.0 (- (ScalableFloat|GetValueatLevel (Variables|Projectile|GetProjectileAimMismatchMax) (Math|Conversions|ToFloat(Integer) (Variables|Default|GetMaxTeamSize))) (ScalableFloat|GetValueatLevel (Variables|Projectile|GetProjectileAimMismatchMin) (Math|Conversions|ToFloat(Integer) (Variables|Default|GetMaxTeamSize))))) (- (ScalableFloat|GetValueatLevel (Variables|Projectile|GetProjectileAimMismatchMax) (Math|Conversions|ToFloat(Integer) (Variables|Default|GetMaxTeamSize))) (ScalableFloat|GetValueatLevel (Variables|Projectile|GetProjectileAimMismatchMin) (Math|Conversions|ToFloat(Integer) (Variables|Default|GetMaxTeamSize))))) (ScalableFloat|GetValueatLevel (Variables|Projectile|GetProjectileAimMismatchMultiplier) (Math|Conversions|ToFloat(Integer) (Variables|Default|GetMaxTeamSize))))) (<= (ScalableFloat|GetValueatLevel (Variables|Projectile|GetProjectileTurnSpeedAffectedByAimMismatch) (Math|Conversions|ToFloat(Integer) (Variables|Default|GetMaxTeamSize))) 0))))

(fn GetProjectileSpawnLocation
  (bind (_returnvalue_x _returnvalue_y) (Math|Vector|BreakVector (Transformation|GetActorForwardVector (CallFunction|GetAvatarasFortPlayerPawn))))
  (return (Utilities|Operators|Add (Transformation|GetActorLocation (CallFunction|GetAvatarasFortPlayerPawn)) (Math|Vector|MakeVector (* _returnvalue_x (ScalableFloat|GetValueatLevel (Variables|Projectile|GetProjectileSpawnDistanceXOffset) (Math|Conversions|ToFloat(Integer) (Variables|Default|GetMaxTeamSize)))) (* _returnvalue_y (ScalableFloat|GetValueatLevel (Variables|Projectile|GetProjectileSpawnDistanceYOffset) (Math|Conversions|ToFloat(Integer) (Variables|Default|GetMaxTeamSize)))) 1e+02))))

(fn InitAbility
  (bind _maxteamsize (Variables|Default|GetMaxTeamSize))
  (bind _returnvalue (Math|Conversions|ToFloat(Integer) _maxteamsize))
  (bind _aipawn (Variables|Default|GetAIPawn))
  (bind _returnvalue_2 (Actor|GetComponentbyClass _aipawn "/Script/EnergyRuntime.FortComponent_Energy"))
  (bind _energycost (Variables|Default|GetEnergyCost))
  (bind _returnvalue_3 (ScalableFloat|GetValueatLevel _energycost _returnvalue))
  (bind _output_get (Variables|Default|SetAIPawn (CallFunction|GetAvatarasFortPlayerPawn)))
  (bind _targetactor (Class|BPFLAICommonFunctions|GetTargetActorfromBlackboard _output_get))
  (bind _output_get_2 (Variables|Default|SetTargetPawn _targetactor))
  (bind _asbp_player_pawn_non_participant_vivid_razor_coffee_puff (Utilities|Casting|CastToBP_PlayerPawn_NonParticipant_VividRazor_CoffeePuff (Variables|Default|GetAIPawn))
    (:then
      (Class|BPPlayerPawnNonParticipantVividRazorCoffeePuff|UpdateLastKnownValidTargetLocation _asbp_player_pawn_non_participant_vivid_razor_coffee_puff (Pawn|GetNavAgentLocation _output_get_2)))
    (:CastFailed
      _))
  (bind _asfort_athena_aibot_controller (Utilities|Casting|CastToFortAthenaAIBotController (Pawn|GetController (Variables|Default|GetAIPawn)))
    (:then
      (bind _output_get_3 (Variables|Default|SetAIBotController _asfort_athena_aibot_controller))
      (AI|PushBehaviorControl _output_get_3 -2097152)
      (Class|FortAthenaAIBotController|StopFire (Variables|Default|GetAIBotController))
      (Class|FortAthenaAIBotController|ResumeTargetFocusing (Variables|Default|GetAIBotController))
      (Energy|UseEnergy _returnvalue_2 _returnvalue_3 true 0)
      (Development|PrintString (Utilities|String|Append "[GA_Hookshot_CoffeePuff] Spending " (Utilities|String|ToString(Float) _returnvalue_3) " energy! Current energy: " (Utilities|String|ToString(Float) (Energy|GetCurrentEnergyPercentage _returnvalue_2))) (Variables|Debug|GetDebugOutput) (Variables|Debug|GetDebugOutput) "(R=1.000000,G=0.467546,B=0.027254,A=1.000000)" 3.0)
      (Variables|Aiming|SetAimLocked false)
      (Variables|Projectile|SetHookshotProjectile 0))
    (:CastFailed
      (Ability|EndAbility))))

(fn CanActivateAbility
  (params (ActorInfo :type Gameplay%20Ability%20Actor%20Info%20Structure) (Handle :type Gameplay%20Ability%20Spec%20Handle%20Structure))
  (bind _asfort_pawn (Utilities|Casting|CastToFortPawn (Utilities|Struct|BreakGameplayAbilityActorInfo ActorInfo)))
  (return 0 (and (not (GameplayTags|HasAnyMatchingGameplayTags _asfort_pawn (GameplayTags|MakeGameplayTagContainerfromArray (Utilities|Array|MakeArray "(TagName=\"Gameplay.Status.Knockback\")")))) (not (AI|Components|NavMovement|IsFalling (Variables|/Script/Engine.Character|GetCharacterMovement _asfort_pawn))))))

(fn PlayRecallSound
  (GameplayCue|ExecuteGameplayCue (Variables|Default|GetAIPawn) (Variables|Cues|GetRecallEffectTag)))

(fn SetLockedFocusPosition
  (bind (_returnvalue_x _returnvalue_y _returnvalue_z) (Math|Vector|BreakVector (Transformation|GetActorLocation (Variables|Default|GetTargetPawn))))
  (bind _aibotcontroller (Variables|Default|GetAIBotController))
  (bind _output_get (Variables|Aiming|SetLockedFocusPosition (Math|Vector|MakeVector _returnvalue_x _returnvalue_y (+ _returnvalue_z (ScalableFloat|GetValueatLevel (Variables|Aiming|GetAimZOffset) (Math|Conversions|ToFloat(Integer) (Variables|Default|GetMaxTeamSize)))))))
  (Class|FortAthenaAIBotController|PauseTargetFocusing _aibotcontroller)
  (AI|SetFocalPoint _aibotcontroller _output_get)
  (Variables|Aiming|SetAimLocked true))
