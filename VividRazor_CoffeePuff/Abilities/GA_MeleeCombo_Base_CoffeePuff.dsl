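; Base melee combo ability for the CoffeePuff AI boss, gated by target range, height and movement state.
; Takes over bot behavior, then chains data-driven attacks, each spending energy and playing a montage with distance-scaled forward root motion.
; Montage notifies trigger async targeting for damage, optional knockback and a squad-size-scaled celebration effect, or allow early outs.
; The combo ends when attacks or energy run out, or the target leaves range.
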(event Custom|ProcessTargets
  (for _array_element 0
    (Ability|ExecuteGameplayCueOnOwner)))

(event Ability|EventOnEndAbility
  (params (bWasCancelled :type Boolean))
  (AI|PopBehaviorControl (Variables|Default|GetAIBotController)))

(event Ability|EventActivateAbility
  (bind (_gamestate _playlist) (Class|FortAsyncActionCurrentPlaylistReady|PlaylistDataReadyAsync)
    (:OnReady
      (Variables|Default|SetMaxTeamSize (Variables|/Script/FortniteGame.FortPlaylist|GetMaxSquadSize _playlist))
      (CallFunction|InitAbility)
      (Ability|CommitAbility)
      (CallFunction|PerformNextAttack))))

(event Custom|PerformNextAttack
  (bind _currentattackdata (Variables|Default|GetCurrentAttackData))
  (bind (_earlyoutenabled_23_29ca7eda436040374346b88e9a656943 _earlyoutdelay_25_63e5eaf1437fbad377bfcdba7542aef8) (Utilities|Struct|BreakSCoffeePuffMeleeComboAttack _currentattackdata))
  (bind _returnvalue (ScalableFloat|GetValueatLevel _earlyoutdelay_25_63e5eaf1437fbad377bfcdba7542aef8))
  (Variables|Default|SetCurrentAttackData (Utilities|Array|Get(acopy) (Variables|Default|GetAttacks) (Variables|Default|GetCurrentAttackIndex)))
  (CallFunction|UseEnergy)
  (bind _forwardrootmotionstrength (CallFunction|CalculateForwardRootMotionStrength))
  (Ability|Tasks|ApplyRootMotionConstantForce "None" (Transformation|GetActorForwardVector (Variables|Default|GetAIPawn)) _forwardrootmotionstrength (ScalableFloat|GetValueatLevel (Utilities|Struct|BreakSCoffeePuffMeleeComboAttack (Variables|Default|GetCurrentAttackData))) false 0 "SetVelocity")
  (bind (_asynctaskproxy _notifytag) (Ability|Tasks|PlayMontageAndWaitNotify "None" (Utilities|Struct|BreakSCoffeePuffMeleeComboAttack (Variables|Default|GetCurrentAttackData)))
    (:then
      (Variables|Default|SetCurrentlyAdvancingAttacks false))
    (:OnCompleted
      (if (not (Variables|Default|GetWaitingForEarlyOutDelay))
        _))
    (:OnInterrupted
      (if (not (Variables|Default|GetCurrentlyAdvancingAttacks)) :!label _label
        (Ability|EndAbility)))
    (:OnCancelled
      (goto _label))
    (:OnNotifyBegin
      (if (GameplayTags|MatchesTag _notifytag "(TagName=\"Targeting.ApplicationTag\")")
        (CallFunction|HandleMontageNotify)
        (CallFunction|DealDamage)
        (else
          (switch Utilities|FlowControl|Switch|SwitchonGameplayTag _notifytag
            (:Ability.VividRazor.Boss.CoffeePuff.MeleeCombo.EarlyOut
              (if (> (ScalableFloat|GetValueatLevel _earlyoutenabled_23_29ca7eda436040374346b88e9a656943) 0)
                (if (> _returnvalue 0)
                  (Ability|Tasks|WaitDelay _returnvalue
                    (:then
                      (Variables|Default|SetWaitingForEarlyOutDelay true))
                    (:OnFinish
                      (Variables|Default|SetWaitingForEarlyOutDelay false)))))))))))
  (Math|Integer|IncrementInt (Variables|Default|GetCurrentAttackIndex)
    (:%20%20
      (if (or (>= (Variables|Default|GetCurrentAttackIndex) (Utilities|Array|Length (Variables|Default|GetAttacks))) (CallFunction|ShouldStopCombo))
        _
        (elif (GameplayTags|HasMatchingGameplayTag (Variables|Default|GetAIPawn) "(TagName=\"Event.Ability.VividRazor.CoffeePuff.EnergyDepleted\")")
          _
          (else
            (Variables|Default|SetCurrentlyAdvancingAttacks true)
            (CallFunction|PerformNextAttack))))
      (Ability|EndAbility))))

(event Custom|DealDamage
  (bind (_applyknockback_32_9119a17c4591e1ac9be7bf9a54685e4a _celebrationchance_38_f6942eb24bb3e4acb11e90afd7e16448) (Utilities|Struct|BreakSCoffeePuffMeleeComboAttack (Variables|Default|GetCurrentAttackData)))
  (bind _returnvalue (GameInstanceSubsystems|GetTargetingSubsystem))
  (bind _aipawn (Variables|Default|GetAIPawn))
  (Utilities|Array|Clear (Variables|Default|GetLatestTargets))
  (bind (_asynctaskproxy _targetinghandle) (Class|AsyncActionPerformTargeting|PerformTargetingAsyncAction _aipawn (Utilities|Struct|BreakSCoffeePuffMeleeComboAttack (Variables|Default|GetCurrentAttackData)))
    (:Targeted
      (if (> (ScalableFloat|GetValueatLevel (Utilities|Struct|BreakSCoffeePuffMeleeComboAttack (Variables|Default|GetCurrentAttackData))) 0)
        (Class|BPFLAICommonFunctions|AddUniqueGameplayTag (Variables|Default|GetAIPawn) "(TagName=\"Ability.VividRazor.Boss.CoffeePuff.MeleeCombo.ApplyKnockback\")"))
      (Utilities|Array|AppendArray (Variables|Default|GetLatestTargets) (TargetingSystem|TargetingResults|GetTargetingResults _returnvalue _targetinghandle))
      (Ability|Tasks|ApplyGameplayEffectToHitResultsAsync (Ability|GetActivatingAbilityComponent) (TargetingSystem|TargetingResults|GetTargetingResults _returnvalue _targetinghandle) (Utilities|Struct|BreakSCoffeePuffMeleeComboAttack (Variables|Default|GetCurrentAttackData)) 0 1 (Class|AthenaFunctionLibrary|MakeSimpleFortEffectContext _aipawn _aipawn))
      (Ability|GameplayEffect|RemoveLooseGameplayTags (Variables|Default|GetAIBotController) (GameplayTags|MakeGameplayTagContainerfromArray (Utilities|Array|MakeArray "(TagName=\"Ability.VividRazor.Boss.CoffeePuff.MeleeCombo.ApplyKnockback\")")))
      (if (< (Math|Random|RandomFloat) (ScalableFloat|GetValueatLevel _celebrationchance_38_f6942eb24bb3e4acb11e90afd7e16448 (Math|Conversions|ToFloat(Integer) (Variables|Default|GetMaxTeamSize))))
        (Ability|Tasks|ApplyGameplayEffectToHitResultsAsync (Ability|GetActivatingAbilityComponent) (Variables|Default|GetLatestTargets) (Variables|Default|GetCelebrationEffect) 0 1 (Class|AthenaFunctionLibrary|MakeSimpleFortEffectContext (Variables|Default|GetAIPawn) (Variables|Default|GetAIPawn)))))))

(fn CalculateForwardRootMotionStrength
  (bind (_rootmotionminscaledistance_18_8e1517634ffeb5ecd854cbb992658d8f _rootmotionmaxscaledistance_20_cd2c32ab41709a72d2efafb09667063d) (Utilities|Struct|BreakSCoffeePuffMeleeComboAttack (Variables|Default|GetCurrentAttackData)))
  (return (* (ScalableFloat|GetValueatLevel (Utilities|Struct|BreakSCoffeePuffMeleeComboAttack (Variables|Default|GetCurrentAttackData))) (/ (Math|Float|Clamp(Float) (- (Transformation|GetDistanceTo (Variables|Default|GetAIPawn) (Variables|Default|GetTargetPawn)) (ScalableFloat|GetValueatLevel _rootmotionminscaledistance_18_8e1517634ffeb5ecd854cbb992658d8f)) 0.0 (ScalableFloat|GetValueatLevel _rootmotionmaxscaledistance_20_cd2c32ab41709a72d2efafb09667063d)) (ScalableFloat|GetValueatLevel _rootmotionmaxscaledistance_20_cd2c32ab41709a72d2efafb09667063d)))))

(fn HandleMontageNotify
  (params (Tag :type Gameplay%20Tag%20Structure))
  (Ability|ExecuteGameplayCueOnOwner Tag))

(fn InitAbility
  (Variables|Default|SetCurrentAttackIndex 0)
  (Variables|Default|SetWaitingForEarlyOutDelay false)
  (bind _output_get (Variables|Default|SetAIPawn (CallFunction|GetAvatarasFortPlayerPawn)))
  (bind _targetactor (Class|BPFLAICommonFunctions|GetTargetActorfromBlackboard _output_get))
  (bind _output_get_2 (Variables|Default|SetTargetPawn _targetactor))
  (bind _asbp_player_pawn_non_participant_vivid_razor_coffee_puff (Utilities|Casting|CastToBP_PlayerPawn_NonParticipant_VividRazor_CoffeePuff (Variables|Default|GetAIPawn))
    (:then
      (Class|BPPlayerPawnNonParticipantVividRazorCoffeePuff|UpdateLastKnownValidTargetLocation _asbp_player_pawn_non_participant_vivid_razor_coffee_puff (Pawn|GetNavAgentLocation _output_get_2)))
    (:CastFailed
      _))
  (bind _asfort_athena_aibot_controller (Utilities|Casting|CastToFortAthenaAIBotController (Pawn|GetController (Variables|Default|GetAIPawn)))
    (:then
      (bind _output_get_3 (Variables|Default|SetAIBotController _asfort_athena_aibot_controller))
      (AI|PushBehaviorControl _output_get_3 -2097152)
      (Class|FortAthenaAIBotController|StopFire (Variables|Default|GetAIBotController)))
    (:CastFailed
      (Ability|EndAbility))))

(fn IsTargetOutOfVerticalRange
  (return (and (AI|Components|NavMovement|IsMovingonGround (Variables|/Script/Engine.Character|GetCharacterMovement (Variables|Default|GetTargetPawn))) (>= (Transformation|GetVerticalDistanceTo (Variables|Default|GetAIPawn) (Variables|Default|GetTargetPawn)) (ScalableFloat|GetValueatLevel (Variables|Default|GetMaxAllowedZDistanceToTarget) (Math|Conversions|ToFloat(Integer) (Variables|Default|GetMaxTeamSize)))))))

(fn CanActivateAbility
  (params (ActorInfo :type Gameplay%20Ability%20Actor%20Info%20Structure) (Handle :type Gameplay%20Ability%20Spec%20Handle%20Structure))
  (bind _returnvalue (CallFunction|GetAvatarasFortPlayerPawn))
  (bind _targetactor (Class|BPFLAICommonFunctions|GetTargetActorfromBlackboard _returnvalue))
  (bind _output_get (LocalVariables|SetCurrentTarget _targetactor))
  (bind _returnvalue_2 (GameplayEffects|GetActiveEffectswithAllTags (Ability|GetFortAbilitySystemComponent _returnvalue) (GameplayTags|MakeGameplayTagContainerfromArray (Utilities|Array|MakeArray "(TagName=\"Ability.VividRazor.Boss.CoffeePuff.Hookshot.BeingPulled\")"))))
  (return 0 (Math|Boolean|ANDBoolean (or (> (Utilities|Array|Length _returnvalue_2) 0) (<= (Transformation|GetDistanceTo _output_get _returnvalue) 1.1e+03)) (not (GameplayTags|HasAnyMatchingGameplayTags (CallFunction|GetAvatarasFortPlayerPawn) (GameplayTags|MakeGameplayTagContainerfromArray (Utilities|Array|MakeArray "(TagName=\"Gameplay.Status.Knockback\")")))) (not (AI|Components|NavMovement|IsFalling (Variables|/Script/Engine.Character|GetCharacterMovement (CallFunction|GetAvatarasFortPlayerPawn)))) (not (and (AI|Components|NavMovement|IsMovingonGround (Variables|/Script/Engine.Character|GetCharacterMovement (LocalVariables|GetCurrentTarget))) (>= (Transformation|GetVerticalDistanceTo (CallFunction|GetAvatarasFortPlayerPawn) (LocalVariables|GetCurrentTarget)) (ScalableFloat|GetValueatLevel (Variables|Default|GetMaxAllowedZDistanceToTarget) (Math|Conversions|ToFloat(Integer) (Variables|Default|GetMaxTeamSize)))))))))

(fn ShouldStopCombo
  (if (>= (Variables|Default|GetCurrentAttackIndex) (Utilities|Array|Length (Variables|Default|GetAttacks)))
    (return true)
    (else
      (return (or (> (Transformation|GetDistanceTo (Variables|Default|GetAIPawn) (Variables|Default|GetTargetPawn)) (ScalableFloat|GetValueatLevel (Utilities|Struct|BreakSCoffeePuffMeleeComboAttack (Utilities|Array|Get(aref) (Variables|Default|GetAttacks) (Variables|Default|GetCurrentAttackIndex))))) (CallFunction|IsTargetOutOfVerticalRange))))))

(fn UseEnergy
  (bind _currentattackdata (Variables|Default|GetCurrentAttackData))
  (bind _energycost_35_5c429e1c401f13ae321c4683a3801fe9 (Utilities|Struct|BreakSCoffeePuffMeleeComboAttack _currentattackdata))
  (bind _aipawn (Variables|Default|GetAIPawn))
  (bind _returnvalue (Actor|GetComponentbyClass _aipawn "/Script/EnergyRuntime.FortComponent_Energy"))
  (bind _returnvalue_2 (ScalableFloat|GetValueatLevel _energycost_35_5c429e1c401f13ae321c4683a3801fe9))
  (Energy|UseEnergy _returnvalue _returnvalue_2 true 0)
  (Development|PrintString (Utilities|String|Append "[GA_MeleeCombo_Base_CoffeePuff] Spending " (Utilities|String|ToString(Float) _returnvalue_2) " energy! Current energy: " (Utilities|String|ToString(Float) (Energy|GetCurrentEnergyPercentage _returnvalue))) (Variables|Default|GetDebugOutput) (Variables|Default|GetDebugOutput) "(R=1.000000,G=0.467546,B=0.027254,A=1.000000)" 3.0))

(fn WasTargetJustHookshotPulled
  (bind _targetactor (Class|BPFLAICommonFunctions|GetTargetActorfromBlackboard (CallFunction|GetAvatarasFortPlayerPawn)))
  (bind _returnvalue (GameplayEffects|GetActiveEffectswithAllTags (Ability|GetFortAbilitySystemComponent _targetactor) (GameplayTags|MakeGameplayTagContainerfromArray (Utilities|Array|MakeArray "(TagName=\"Ability.VividRazor.Boss.CoffeePuff.Hookshot.BeingPulled\")"))))
  (return (> (Utilities|Array|Length _returnvalue) 0)))
