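; Gameplay ability that teleports a falling CoffeePuff AI pawn back onto navmesh.
; Activation checks fall speed or vertical distance from spawn, optionally requiring no navmesh nearby, then plays a tell cue and flags recovery on the blackboard.
; Teleport execution projects the target's last known location, or spawn, onto navmesh, facing any valid target.
; Retries use growing probe extents, random known locations and spawn fallback, logging attempted destinations at the limit.
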(event Ability|EventActivateAbility
  (|CallParentFunction)
  (GameplayCue|ExecuteGameplayCue (Variables|Default|GetAIPawn) (Variables|Default|GetRecoveryTellCue))
  (Variables|Default|SetTotalAttempts 0)
  (Variables|Default|SetCurrentProbeMultiplierXY 1.0)
  (Variables|Default|SetCurrentProbeMultiplierZ 1.0)
  (Utilities|Array|Clear (Variables|Default|GetAllAttemptedTeleportDestinations))
  (AI|Components|Blackboard|SetValueasBool (AI|GetBlackboard (Variables|Default|GetAIPawn)) (Utilities|Name|MakeLiteralName "CoffeePuff_ActivelyRecovering") true)
  (Variables|Default|SetRandomizeAttemptedLocations false)
  (Variables|Default|SetNavmeshFound false)
  (Variables|Default|SetTeleportSucceeded false)
  (Utilities|Array|Clear (Variables|Default|GetAllKnownValidLocations)))

(event Ability|EventOnEndAbility
  (params (bWasCancelled :type Boolean))
  (|CallParentFunction bWasCancelled)
  (AI|Components|Blackboard|SetValueasBool (AI|GetBlackboard (Variables|Default|GetAIPawn)) (Utilities|Name|MakeLiteralName "CoffeePuff_ActivelyRecovering") false))

(event EventOnExecuteTeleport
  (bind _allattemptedteleportdestinations (Variables|Default|GetAllAttemptedTeleportDestinations))
  (bind _aipawn (Variables|Default|GetAIPawn))
  (bind _allknownvalidlocations (Class|BPPlayerPawnNonParticipantVividRazorCoffeePuff|GetAllKnownValidLocations _aipawn :!class /VividRazor_CoffeePuff/NPCs/CoffeePuff/Pawn/BP_PlayerPawn_NonParticipant_VividRazor_CoffeePuff.BP_PlayerPawn_NonParticipant_VividRazor_CoffeePuff_C))
  (Variables|Default|SetAllKnownValidLocations _allknownvalidlocations)
  (bind _haslastknownlocation (Class|BPPlayerPawnNonParticipantVividRazorCoffeePuff|HasValidLastKnownTargetLocation _aipawn))
  (if _haslastknownlocation
    (Variables|Default|SetCurrentTeleportProbePosition (AI|Components|Blackboard|GetValueasVector (AI|GetBlackboard (Variables|Default|GetAIBotController)) (Utilities|Name|MakeLiteralName "CoffeePuff_TargetLastKnownLocation")))
    (else
      (Variables|Default|SetCurrentTeleportProbePosition (AI|Components|Blackboard|GetValueasVector (AI|GetBlackboard (Variables|Default|GetAIBotController)) (Utilities|Name|MakeLiteralName "CoffeePuff_SpawnLocation")))))
  (bind (_projectedlocation _navmeshfound) (CallFunction|ProjectLocationtoNavmesh (Variables|Default|GetCurrentTeleportProbePosition)))
  (bind (_location _rotation) (CallFunction|CalculateDestinationLocationandRotation _projectedlocation))
  (Utilities|Array|Add (Variables|Default|GetAllAttemptedTeleportDestinations) _location)
  (Variables|Default|SetNavmeshFound _navmeshfound)
  (if _navmeshfound
    (bind _returnvalue (Transformation|Teleport (Variables|Default|GetAIPawn) _location _rotation))
    (bind _output_get (Variables|Default|SetTeleportSucceeded _returnvalue))
    (if (not _output_get)
      (Math|Integer|IncrementInt (Variables|Default|GetTotalAttempts)
        (:%20%20
          (Variables|Default|SetRandomizeAttemptedLocations true)
          (if (not (>= (Variables|Default|GetTotalAttempts) (+ (ScalableFloat|GetValueatLevel (Variables|Default|GetTargetLocationMaxAttempts)) (ScalableFloat|GetValueatLevel (Variables|Default|GetSpawnLocationMaxAttempts)))))
            (Utilities|FlowControl|DelayUntilNextTick)
            (Variables|Default|SetCurrentProbeMultiplierXY (* (Variables|Default|GetCurrentProbeMultiplierXY) (ScalableFloat|GetValueatLevel (Variables|Default|GetRecoveryProbe_IterationMultiplierXY))))
            (Variables|Default|SetCurrentProbeMultiplierZ (* (Variables|Default|GetCurrentProbeMultiplierZ) (ScalableFloat|GetValueatLevel (Variables|Default|GetRecoveryProbe_IterationMultiplierZ))))
            (if (Variables|Default|GetRandomizeAttemptedLocations)
              (bind _knownlocation (CallFunction|GetRandomKnownLocation true))
              (Variables|Default|SetCurrentTeleportProbePosition _knownlocation))
            (if (>= (Variables|Default|GetTotalAttempts) (ScalableFloat|GetValueatLevel (Variables|Default|GetTargetLocationMaxAttempts)))
              (Variables|Default|SetCurrentTeleportProbePosition (AI|Components|Blackboard|GetValueasVector (AI|GetBlackboard (Variables|Default|GetAIBotController)) (Utilities|Name|MakeLiteralName "CoffeePuff_SpawnLocation")))))
          (Variables|Default|SetErrorString (Utilities|String|Append "[CoffeePuff_RecoveryTeleport] Reached maximum number [" (Utilities|String|ToString(Integer) (Variables|Default|GetTotalAttempts)) "] of recovery attempts. Attempted locations:  "))
          (for (_array_element _array_index) _allattemptedteleportdestinations
            (bind _output_get_2 (Variables|Default|SetErrorString (Utilities|String|Append (Variables|Default|GetErrorString) (Utilities|String|ToString(Vector) _array_element))))
            (Variables|Default|SetErrorString (Utilities|String|Append _output_get_2 (Utilities|String|SelectString "" " | " (== _array_index (Utilities|Array|Length _allattemptedteleportdestinations))))))
          (Fortnite|FortShippingLog (Variables|Default|GetAIPawn) (Variables|Default|GetErrorString))
          (Ability|EndAbility))))
    (Class|FortAthenaAIBotController|PauseTargetFocusing (Variables|Default|GetAIBotController))
    (AI|Components|Blackboard|SetValueasBool (AI|GetBlackboard (Variables|Default|GetAIPawn)) (Utilities|Name|MakeLiteralName "CoffeePuff_ActivelyRecovering") false)))

(fn CalculateDestinationLocationAndRotation
  (params (InputLocation :type Vector))
  (bind (_x _y _z) (Math|Rotator|BreakRotIntoAxes (Transformation|GetActorRotation (Variables|Default|GetAIPawn))))
  (return (+ (Variables|GetInputLocation) (Math|Vector|MakeVector (- (Components|Capsule|GetScaledCapsuleRadius (Variables|/Script/Engine.Character|GetCapsuleComponent (Variables|Default|GetAIPawn)))) 0.0 (Components|Capsule|GetScaledCapsuleHalfHeight (Variables|/Script/Engine.Character|GetCapsuleComponent (Variables|Default|GetAIPawn))))) (Math|Rotator|MakeRotationfromAxes (Math|Vector|SelectVector (- (Transformation|GetActorLocation (Variables|Default|GetTargetPawn)) (Variables|GetInputLocation)) _x (Utilities|IsValid (Variables|Default|GetTargetPawn) :!class /Script/Engine.KismetSystemLibrary)) _y _z)))

(fn GetRandomKnownLocation
  (params (RemoveFromPool :type Boolean :default true))
  (LocalVariables|SetChosenLocation (Utilities|Array|RandomArrayItem (Variables|Default|GetAllKnownValidLocations)))
  (if (and (Variables|GetRemoveFromPool) (> (Utilities|Array|Length (Variables|Default|GetAllKnownValidLocations)) 1))
    (Utilities|Array|RemoveItem (Variables|Default|GetAllKnownValidLocations) (LocalVariables|GetChosenLocation))
    (else
      _))
  (return (LocalVariables|GetChosenLocation)))

(fn CanActivateAbility
  (params (ActorInfo :type Gameplay%20Ability%20Actor%20Info%20Structure) (Handle :type Gameplay%20Ability%20Spec%20Handle%20Structure))
  (bind (_returnvalue_x _returnvalue_y _returnvalue_z) (Math|Vector|BreakVector (AI|Components|Blackboard|GetValueasVector (AI|GetBlackboard (CallFunction|GetAvatarasFortPlayerPawn)) (Utilities|Name|MakeLiteralName "CoffeePuff_SpawnLocation"))))
  (bind (_returnvalue_x_2 _returnvalue_y_2 _returnvalue_z_2) (Math|Vector|BreakVector (Pawn|GetNavAgentLocation (CallFunction|GetAvatarasFortPlayerPawn))))
  (bind (_projectedlocation _returnvalue) (AI|Navigation|ProjectPointtoNavigation (Pawn|GetNavAgentLocation (CallFunction|GetAvatarasFortPlayerPawn)) (Fortnite|AI|GetNavigationDataforActor (CallFunction|GetAvatarasFortPlayerPawn)) "/Game/Athena/AI/Phoebe/NavFilter/NavFilter_Phoebe.NavFilter_Phoebe_C" "200.000000,200.000000,500.000000"))
  (return 0 (Math|Boolean|ANDBoolean (AI|Components|NavMovement|IsFalling (Variables|/Script/Engine.Character|GetCharacterMovement (CallFunction|GetAvatarasFortPlayerPawn))) (or (< (.z (Pawn|Components|CharacterMovement|GetLastUpdateVelocity (Variables|/Script/Engine.Character|GetCharacterMovement (CallFunction|GetAvatarasFortPlayerPawn)))) (ScalableFloat|GetValueatLevel (Variables|Default|GetActivationZVelocityMin))) (and (> (ScalableFloat|GetValueatLevel (Variables|Default|GetCheckVerticalDistanceFromSpawn)) 0) (> (Math|Float|Absolute(Float) (- _returnvalue_z _returnvalue_z_2)) (ScalableFloat|GetValueatLevel (Variables|Default|GetMaxVerticalDistanceFromSpawn))))) (select (> (ScalableFloat|GetValueatLevel (Variables|Default|GetCheckForFloorUnderPawn)) 0) (not _returnvalue) true))))

(fn ProjectLocationToNavmesh
  (params (Location :type Vector))
  (bind (_projectedlocation _returnvalue) (AI|Navigation|ProjectPointtoNavigation Location (Fortnite|AI|GetNavigationDataforActor (Variables|Default|GetAIPawn)) "/Game/Athena/AI/Phoebe/NavFilter/NavFilter_Phoebe.NavFilter_Phoebe_C" (Math|Vector|MakeVector (* (ScalableFloat|GetValueatLevel (Variables|Default|GetRecoveryProbe_BaseExtentX)) (Variables|Default|GetCurrentProbeMultiplierXY)) (* (ScalableFloat|GetValueatLevel (Variables|Default|GetRecoveryProbe_BaseExtentY)) (Variables|Default|GetCurrentProbeMultiplierXY)) (* (ScalableFloat|GetValueatLevel (Variables|Default|GetRecoveryProbe_BaseExtentZ)) (Variables|Default|GetCurrentProbeMultiplierZ)))))
  (return _projectedlocation _returnvalue))

(fn SelectDestinationCenter
  (bind _haslastknownlocation (Class|BPPlayerPawnNonParticipantVividRazorCoffeePuff|HasValidLastKnownTargetLocation (Variables|Default|GetAIPawn)))
  (if _haslastknownlocation
    (return (AI|Components|Blackboard|GetValueasVector (AI|GetBlackboard (Variables|Default|GetAIPawn)) (Utilities|Name|MakeLiteralName "CoffeePuff_TargetLastKnownLocation")))
    (else
      (return (AI|Components|Blackboard|GetValueasVector (AI|GetBlackboard (Variables|Default|GetAIPawn)) (Utilities|Name|MakeLiteralName "CoffeePuff_SpawnLocation"))))))
