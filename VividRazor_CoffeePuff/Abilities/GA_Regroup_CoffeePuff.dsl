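; Gameplay ability that makes the CoffeePuff AI pawn regroup when tagged EnergyDepleted, provided it is neither knocked back nor falling.
; On activation it records the target's last known location, pushes behavior control and stops firing.
; After committing, it plays the regroup montage with a randomly delayed speech bubble.
; Ending the ability adds energy, clears the depleted tag from the bot controller and pops behavior control.
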(event Ability|EventActivateAbility
  (bind (_montageinfo_montagetoplay _montageinfo_animplayrate _montageinfo_animrootmotiontranslationscale _montageinfo_montagesectiontoplay _montageinfo_overridesection) (Utilities|Struct|BreakFortGameplayAbilityMontageInfo (Variables|GetMontageInfo_MontageToPlay)))
  (Development|PrintString "[GA_Regroup_CoffeePuff]: Activating ability!" (Variables|Default|GetDebugOutput) (Variables|Default|GetDebugOutput) (Variables|Default|GetDebugOutputColor) (Variables|Default|GetDebugOutputDuration))
  (CallFunction|InitAbility)
  (bind _returnvalue (Ability|CommitAbility))
  (if _returnvalue
    (Ability|Tasks|PlayMontageAndWaitNotify "CoffeePuff_Regroup" _montageinfo_montagetoplay _montageinfo_animplayrate _montageinfo_overridesection true _montageinfo_animrootmotiontranslationscale
      (:then
        (Ability|Tasks|WaitDelay (Math|Random|RandomFloatinRange 0.2 0.6)
          (:OnFinish
            (Class|BPFLAICommonFunctions|PlaySpeechBubble (Variables|Default|GetAIPawn) (Variables|Default|GetRegroupSpeechTag) true true))))
      (:OnCompleted
        _)
      (:OnInterrupted
        _)
      (:OnCancelled
        _))
    (Ability|EndAbility)
    (else
      (Ability|EndAbility))))

(event Ability|EventOnEndAbility
  (params (bWasCancelled :type Boolean))
  (bind _aibotcontroller (Variables|Default|GetAIBotController))
  (Development|PrintString "[GA_Regroup_CoffeePuff]: Ending ability!" (Variables|Default|GetDebugOutput) (Variables|Default|GetDebugOutput) (Variables|Default|GetDebugOutputColor) (Variables|Default|GetDebugOutputDuration))
  (Energy|AddEnergy (Actor|GetComponentbyClass (Variables|Default|GetAIPawn) "/Script/EnergyRuntime.FortComponent_Energy") 1e+02 0 "(GameplayTags=)")
  (Utilities|IsValid _aibotcontroller
    (:Is%20Valid
      (Class|BPFLAICommonFunctions|RemoveAllInstancesOfGameplayTag _aibotcontroller "(TagName=\"Event.Ability.VividRazor.CoffeePuff.EnergyDepleted\")")
      (AI|PopBehaviorControl _aibotcontroller))))

(fn InitAbility
  (bind _output_get (Variables|Default|SetAIPawn (CallFunction|GetAvatarasFortPlayerPawn)))
  (bind _targetactor (Class|BPFLAICommonFunctions|GetTargetActorfromBlackboard _output_get))
  (bind _output_get_2 (Variables|Default|SetTargetPawn _targetactor))
  (bind _asbp_player_pawn_non_participant_vivid_razor_coffee_puff (Utilities|Casting|CastToBP_PlayerPawn_NonParticipant_VividRazor_CoffeePuff (Variables|Default|GetAIPawn))
    (:then
      (Class|BPPlayerPawnNonParticipantVividRazorCoffeePuff|UpdateLastKnownValidTargetLocation _asbp_player_pawn_non_participant_vivid_razor_coffee_puff (Pawn|GetNavAgentLocation _output_get_2)))
    (:CastFailed
      _))
  (bind _asfort_athena_aibot_controller (Utilities|Casting|CastToFortAthenaAIBotController (Pawn|GetController (Variables|Default|GetAIPawn)))
    (:then
      (bind _output_get_3 (Variables|Default|SetAIBotController _asfort_athena_aibot_controller))
      (AI|PushBehaviorControl _output_get_3 -2097152)
      (Class|FortAthenaAIBotController|StopFire (Variables|Default|GetAIBotController)))
    (:CastFailed
      (Ability|EndAbility))))

(fn CanActivateAbility
  (params (ActorInfo :type Gameplay%20Ability%20Actor%20Info%20Structure) (Handle :type Gameplay%20Ability%20Spec%20Handle%20Structure))
  (return 0 (Math|Boolean|ANDBoolean (GameplayTags|HasMatchingGameplayTag (CallFunction|GetAvatarasFortPlayerPawn) "(TagName=\"Event.Ability.VividRazor.CoffeePuff.EnergyDepleted\")") (not (GameplayTags|HasAnyMatchingGameplayTags (CallFunction|GetAvatarasFortPlayerPawn) (GameplayTags|MakeGameplayTagContainerfromArray (Utilities|Array|MakeArray "(TagName=\"Gameplay.Status.Knockback\")")))) (not (AI|Components|NavMovement|IsFalling (Variables|/Script/Engine.Character|GetCharacterMovement (CallFunction|GetAvatarasFortPlayerPawn)))))))
