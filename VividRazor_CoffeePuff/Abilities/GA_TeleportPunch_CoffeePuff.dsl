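; Gameplay ability for the CoffeePuff AI bot: an energy-costing teleport punch, with tuning scaled by playlist squad size.
; It hides the bot invulnerably, runs EQS around the target or its last known location, and teleports there, falling back to spawn.
; Timed cues telegraph arrival, then an outro montage notify starts repeating targeting sweeps that damage each target once.
; Activation requires a leashed, navigable target and no knockback or falling; ending restores visibility and AI control.
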(event Ability|EventActivateAbility
  (bind _aipawn (Variables|Default|GetAIPawn))
  (bind (_gamestate _playlist) (Class|FortAsyncActionCurrentPlaylistReady|PlaylistDataReadyAsync)
    (:OnReady
      (Variables|Default|SetMaxTeamSize (Variables|/Script/FortniteGame.FortPlaylist|GetMaxSquadSize _playlist))
      (CallFunction|InitAbility)
      (Ability|CommitAbility)
      (if (not (Variables|Default|GetSkipIntroAnimation))
        (bind (_asynctaskproxy _targetdata _applicationtag) (Ability|Tasks|PlayMontageandWaitforTrigger "None" "/VividRazor_CoffeePuff/Assets/Animations/CoffeePuff_HFPunch_Intro_M.CoffeePuff_HFPunch_Intro_M")
          (:Triggered
            (switch Utilities|FlowControl|Switch|SwitchonGameplayTag _applicationtag
              (:Event.Ability.VividRazor.CoffeePuff.BlinkActivation
                _)))))
      (Variables|Default|SetTeleportFXRotation (CallFunction|GetTeleportFXRotation))
      (Ability|ExecuteGameplayCueWithParamsOnOwner (Variables|Default|GetTeleportEffectTag) (Ability|GameplayCue|MakeGameplayCueParameters :NormalizedMagnitude 0.0 :RawMagnitude 1.0 :Location "0,0,0" :Normal (Math|Rotator|GetRotationXVector (Variables|Default|GetTeleportFXRotation)) :GameplayEffectLevel 1 :AbilityLevel 1 :bReplicateLocationWhenUsingMinimalRepProxy false))
      (Rendering|SetActorHiddenInGame _aipawn true)
      (Variables|/Script/FortniteGame.FortPawn|SetbIsInvulnerable true _aipawn)
      (if (> (ScalableFloat|GetValueatLevel (Variables|Default|GetPreTeleportDelay) (Math|Conversions|ToFloat(Integer) (Variables|Default|GetMaxTeamSize))) 0)
        (Utilities|Time|SetTimerbyEvent (EventDispatchers|CreateEvent :!event Custom|StartTeleportProcess) (ScalableFloat|GetValueatLevel (Variables|Default|GetPreTeleportDelay) (Math|Conversions|ToFloat(Integer) (Variables|Default|GetMaxTeamSize))))
        (else
          (CallFunction|StartTeleportProcess))))))

(event Custom|StartTeleportProcess
  (bind _destinationcenter (CallFunction|SelectDestinationCenter))
  (AI|Components|Blackboard|SetValueasVector (AI|GetBlackboard (Variables|Default|GetAIPawn)) (Utilities|Name|MakeLiteralName "CoffeePuff_TeleportPunchDestination") _destinationcenter)
  (bind _returnvalue (AI|EQS|RunEQSQuery (Variables|Default|GetDestinationQueryTemplate) (Variables|Default|GetAIPawn) (select (Variables|Default|GetEQSGetSingleBestResult) "RandomBest5Pct" "AllMatching")))
  (AI|EQS|BindEventtoOnQueryFinished _returnvalue (EventDispatchers|CreateEvent :!event Custom|OnEQSQueryFinished)))

(event Custom|OnEQSQueryFinished
  (params (QueryInstance :type EQS%20Query%20Instance%20Object%20Reference) (QueryStatus :type EEnvQueryStatus%20Enum :default "Processing"))
  (bind _aipawn (Variables|Default|GetAIPawn))
  (if (Utilities|Enum|Equal(Enum) QueryStatus "Success")
    (CallFunction|SelectTeleportPoint QueryInstance)
    (else
      (Fortnite|FortShippingLog (Variables|Default|GetAIPawn) (Utilities|String|Append "[CoffeePuff_TeleportPunch] Failed to find a valid EQS position near the destination. Falling back to spawn location! EQS center position: " (Utilities|String|ToString(Vector) (AI|Components|Blackboard|GetValueasVector (AI|GetBlackboard (Variables|Default|GetAIPawn)) (Utilities|Name|MakeLiteralName "CoffeePuff_TeleportPunchDestination")))))
      (Variables|Default|SetTeleportTargetLocation (AI|Components|Blackboard|GetValueasVector (AI|GetBlackboard (Variables|Default|GetAIPawn)) (Utilities|Name|MakeLiteralName "CoffeePuff_SpawnLocation")))))
  (Variables|Default|SetBeforeTeleportLocation (Transformation|GetWorldLocation (Variables|/Script/Engine.Character|GetCapsuleComponent (Variables|Default|GetAIPawn))))
  (CallFunction|OnExecuteTeleport)
  (GameplayCue|ExecuteGameplayCue (Variables|Default|GetAIPawn) (Variables|Default|GetTelegraphEffectTag))
  (Ability|ExecuteGameplayCueWithParamsOnOwner (Variables|Default|GetSpeedTrailEffectTag) (Ability|GameplayCue|MakeGameplayCueParameters :NormalizedMagnitude 1.0 :RawMagnitude (+ (ScalableFloat|GetValueatLevel (Variables|Default|GetPostTelegraphDelay) (Math|Conversions|ToFloat(Integer) (Variables|Default|GetMaxTeamSize))) (ScalableFloat|GetValueatLevel (Variables|Default|GetPreTelegraphDelay) (Math|Conversions|ToFloat(Integer) (Variables|Default|GetMaxTeamSize)))) :Location (Variables|Default|GetBeforeTeleportLocation) :Normal "0,0,0" :GameplayEffectLevel 1 :AbilityLevel 1 :bReplicateLocationWhenUsingMinimalRepProxy false))
  (Utilities|FlowControl|Delay (ScalableFloat|GetValueatLevel (Variables|Default|GetPreTelegraphDelay) (Math|Conversions|ToFloat(Integer) (Variables|Default|GetMaxTeamSize))))
  (GameplayCue|ExecuteGameplayCue (Variables|Default|GetAIPawn) (Variables|Default|GetTeleportOutEffectTag))
  (Utilities|FlowControl|Delay (ScalableFloat|GetValueatLevel (Variables|Default|GetPostTelegraphDelay) (Math|Conversions|ToFloat(Integer) (Variables|Default|GetMaxTeamSize))))
  (Ability|ExecuteGameplayCueWithParamsOnOwner (Variables|Default|GetTeleportEffectTag) (Ability|GameplayCue|MakeGameplayCueParameters :NormalizedMagnitude 1.0 :RawMagnitude 0.0 :Location "0,0,0" :Normal (Transformation|GetActorForwardVector (Variables|Default|GetAIPawn)) :GameplayEffectLevel 1 :AbilityLevel 1 :bReplicateLocationWhenUsingMinimalRepProxy false))
  (bind (_asynctaskproxy _notifytag) (Ability|Tasks|PlayMontageAndWaitNotify "Task_PlayMontage" "/VividRazor_CoffeePuff/Assets/Animations/CoffeePuff_HFPunch_Outro_M.CoffeePuff_HFPunch_Outro_M" 1.0 "None" false)
    (:then
      (Utilities|FlowControl|Delay (ScalableFloat|GetValueatLevel (Variables|Default|GetUnhideDelay)))
      (Rendering|SetActorHiddenInGame _aipawn false)
      (Variables|/Script/FortniteGame.FortPawn|SetbIsInvulnerable false _aipawn))
    (:OnCompleted
      _)
    (:OnInterrupted
      _)
    (:OnCancelled
      _)
    (:OnNotifyBegin
      (switch Utilities|FlowControl|Switch|SwitchonGameplayTag _notifytag
        (:Event.Ability.VividRazor.CoffeePuff.BlinkActivation
          (CallFunction|AttemptHit)))))
  (Ability|EndAbility))

(event Custom|AttemptHit
  (bind (_asynctaskproxy _targetinghandle) (Class|AsyncActionPerformTargeting|PerformTargetingAsyncAction (Variables|Default|GetAIPawn) (Variables|Default|GetTargetingPreset))
    (:Targeted
      (Utilities|Array|Clear (Variables|Default|GetNewTargets))
      (for _array_element (TargetingSystem|TargetingResults|GetTargetingResults (GameInstanceSubsystems|GetTargetingSubsystem) _targetinghandle)
        (if (not (Utilities|Array|ContainsItem (Variables|Default|GetAlreadyAffectedTargets) (.HitActor (Collision|BreakHitResult _array_element))))
          (Utilities|Array|Add (Variables|Default|GetNewTargets) _array_element)
          (Utilities|Array|Add (Variables|Default|GetAlreadyAffectedTargets) (.HitActor (Collision|BreakHitResult _array_element)))))
      (Ability|Tasks|ApplyGameplayEffectToHitResultsAsync (Ability|GetActivatingAbilityComponent) (Variables|Default|GetNewTargets) (Variables|Default|GetDamageEffectClass) 0 1 (Class|AthenaFunctionLibrary|MakeSimpleFortEffectContext (Variables|Default|GetAIPawn) (Variables|Default|GetAIPawn)))
      (if (not (Variables|Default|GetSlideCompleted))
        (Utilities|Time|SetTimerbyEvent (EventDispatchers|CreateEvent :!event Custom|AttemptHit) (ScalableFloat|GetValueatLevel (Variables|Default|GetHitRate) (Math|Conversions|ToFloat(Integer) (Variables|Default|GetMaxTeamSize))))))))

(event Ability|EventOnEndAbility
  (params (bWasCancelled :type Boolean))
  (bind _aipawn (Variables|Default|GetAIPawn))
  (bind _visualizationactor (Variables|Default|GetVisualizationActor))
  (bind _aibotcontroller (Variables|Default|GetAIBotController))
  (Rendering|SetActorHiddenInGame _aipawn false)
  (Variables|/Script/FortniteGame.FortPawn|SetbIsInvulnerable false _aipawn)
  (if (Utilities|IsValid _visualizationactor :!class /Script/Engine.KismetSystemLibrary)
    (Actor|DestroyActor _visualizationactor)
    (else
      _))
  (AI|PopBehaviorControl _aibotcontroller)
  (Class|FortAthenaAIBotController|ResumeTargetFocusing _aibotcontroller))

(event Custom|OnExecuteTeleport
  (bind _aipawn (Variables|Default|GetAIPawn))
  (bind _capsulecomponent (Variables|/Script/Engine.Character|GetCapsuleComponent _aipawn))
  (bind _returnvalue (Components|Capsule|GetScaledCapsuleHalfHeight _capsulecomponent))
  (bind _returnvalue_2 (Components|Capsule|GetScaledCapsuleRadius _capsulecomponent))
  (bind _returnvalue_3 (- _returnvalue_2))
  (bind _returnvalue_4 (Math|Vector|MakeVector _returnvalue_3 0.0 _returnvalue))
  (bind _teleporttargetlocation (Variables|Default|GetTeleportTargetLocation))
  (bind _returnvalue_5 (+ _teleporttargetlocation _returnvalue_4))
  (bind _returnvalue_6 (Utilities|String|ToString(Vector) _returnvalue_5))
  (bind (_x _y _z) (Math|Rotator|BreakRotIntoAxes (Transformation|GetActorRotation _aipawn)))
  (bind _debugoutput (Variables|Default|GetDebugOutput))
  (Variables|Default|SetTeleportDestination _returnvalue_5)
  (bind _returnvalue_7 (Transformation|Teleport _aipawn _returnvalue_5 (Math|Rotator|MakeRotationfromAxes (Math|Vector|SelectVector (- (Transformation|GetActorLocation (Variables|Default|GetTargetPawn)) (Variables|Default|GetTeleportTargetLocation)) _x (Utilities|IsValid (Variables|Default|GetTargetPawn) :!class /Script/Engine.KismetSystemLibrary)) _y _z)))
  (if _returnvalue_7
    (Class|FortAthenaAIBotController|PauseTargetFocusing (Variables|Default|GetAIBotController))
    (Development|PrintString (Utilities|String|Append "[BLINK Ability] Succesfully teleported to (" _returnvalue_6 ").") _debugoutput)
    (Variables|Default|SetSuccessfullyTeleported true)
    (else
      (Development|PrintString (Utilities|String|Append "[BLINK Ability] FAILED teleporting to (" _returnvalue_6 ").") _debugoutput))))

(fn ExecuteTeleport
  (bind _aipawn (Variables|Default|GetAIPawn))
  (bind _capsulecomponent (Variables|/Script/Engine.Character|GetCapsuleComponent _aipawn))
  (bind _returnvalue (Components|Capsule|GetScaledCapsuleHalfHeight _capsulecomponent))
  (bind _returnvalue_2 (Components|Capsule|GetScaledCapsuleRadius _capsulecomponent))
  (bind _returnvalue_3 (- _returnvalue_2))
  (bind _returnvalue_4 (Math|Vector|MakeVector _returnvalue_3 0.0 _returnvalue))
  (bind _teleporttargetlocation (Variables|Default|GetTeleportTargetLocation))
  (bind _returnvalue_5 (+ _teleporttargetlocation _returnvalue_4))
  (bind _returnvalue_6 (Utilities|String|ToString(Vector) _returnvalue_5))
  (bind (_x _y _z) (Math|Rotator|BreakRotIntoAxes (Transformation|GetActorRotation _aipawn)))
  (bind _debugoutput (Variables|Default|GetDebugOutput))
  (Variables|Default|SetTeleportDestination _returnvalue_5)
  (bind _returnvalue_7 (Transformation|Teleport _aipawn _returnvalue_5 (Math|Rotator|MakeRotationfromAxes (Math|Vector|SelectVector (- (Transformation|GetActorLocation (Variables|Default|GetTargetPawn)) (Variables|Default|GetTeleportTargetLocation)) _x (Utilities|IsValid (Variables|Default|GetTargetPawn) :!class /Script/Engine.KismetSystemLibrary)) _y _z)))
  (if _returnvalue_7
    (Class|FortAthenaAIBotController|PauseTargetFocusing (Variables|Default|GetAIBotController))
    (Development|PrintString (Utilities|String|Append "[BLINK Ability] Succesfully teleported to (" _returnvalue_6 ").") _debugoutput)
    (else
      (Development|PrintString (Utilities|String|Append "[BLINK Ability] FAILED teleporting to (" _returnvalue_6 ").") _debugoutput))))

(fn GetTeleportFXRotation
  (return (Math|Rotator|FindLookatRotation (Transformation|GetWorldLocation (Variables|/Script/Engine.Character|GetCapsuleComponent (Variables|Default|GetAIPawn))) (Math|Vector|SelectVector (Transformation|GetWorldLocation (Variables|/Script/Engine.Character|GetCapsuleComponent (Variables|Default|GetTargetPawn))) (AI|Components|Blackboard|GetValueasVector (AI|GetBlackboard (Variables|Default|GetAIPawn)) (Utilities|Name|MakeLiteralName "NonCoffeePuff_SpawnLocatione")) (Utilities|IsValid (Variables|Default|GetTargetPawn) :!class /Script/Engine.KismetSystemLibrary)))))

(fn InitAbility
  (bind _maxteamsize (Variables|Default|GetMaxTeamSize))
  (bind _returnvalue (Math|Conversions|ToFloat(Integer) _maxteamsize))
  (bind _aipawn (Variables|Default|GetAIPawn))
  (bind _returnvalue_2 (Actor|GetComponentbyClass _aipawn "/Script/EnergyRuntime.FortComponent_Energy"))
  (bind _energycost (Variables|Default|GetEnergyCost))
  (bind _returnvalue_3 (ScalableFloat|GetValueatLevel _energycost _returnvalue))
  (bind _asbp_player_pawn_non_participant_vivid_razor_coffee_puff (Utilities|Casting|CastToBP_PlayerPawn_NonParticipant_VividRazor_CoffeePuff (CallFunction|GetAvatarasFortPlayerPawn)))
  (bind _output_get (Variables|Default|SetAIPawn _asbp_player_pawn_non_participant_vivid_razor_coffee_puff))
  (bind _asfort_athena_aibot_controller (Utilities|Casting|CastToFortAthenaAIBotController (Pawn|GetController _output_get)))
  (Variables|Default|SetAIBotController _asfort_athena_aibot_controller)
  (bind _targetactor (Class|BPFLAICommonFunctions|GetTargetActorfromBlackboard (Variables|Default|GetAIPawn)))
  (bind _output_get_2 (Variables|Default|SetTargetPawn _targetactor))
  (Utilities|IsValid _output_get_2
    (:Is%20Valid
      (Class|BPPlayerPawnNonParticipantVividRazorCoffeePuff|UpdateLastKnownValidTargetLocation (Variables|Default|GetAIPawn) (Pawn|GetNavAgentLocation _output_get_2)))
    (:Is%20Not%20Valid
      _))
  (AI|PushBehaviorControl (Variables|Default|GetAIBotController) -2097152)
  (Class|FortAthenaAIBotController|StopFire (Variables|Default|GetAIBotController))
  (Variables|Default|SetSlideCompleted false)
  (Utilities|Array|Clear (Variables|Default|GetNewTargets))
  (Utilities|Array|Clear (Variables|Default|GetAlreadyAffectedTargets))
  (Energy|UseEnergy _returnvalue_2 _returnvalue_3 true 0)
  (Development|PrintString (Utilities|String|Append "[GA_TeleportPunch_CoffeePuff] Spending " (Utilities|String|ToString(Float) _returnvalue_3) " energy! Current energy: " (Utilities|String|ToString(Float) (Energy|GetCurrentEnergyPercentage _returnvalue_2))) (Variables|Default|GetDebugOutput) (Variables|Default|GetDebugOutput) "(R=1.000000,G=0.467546,B=0.027254,A=1.000000)" 3.0))

(fn CanActivateAbility
  (params (ActorInfo :type Gameplay%20Ability%20Actor%20Info%20Structure) (Handle :type Gameplay%20Ability%20Spec%20Handle%20Structure))
  (bind _returnvalue (CallFunction|GetAvatarasFortPlayerPawn))
  (bind _targetactor (Class|BPFLAICommonFunctions|GetTargetActorfromBlackboard _returnvalue))
  (bind _output_get (LocalVariables|SetCurrentTarget _targetactor))
  (bind _asfort_athena_aibot_controller (Utilities|Casting|CastToFortAthenaAIBotController (Pawn|GetController _returnvalue)))
  (return 0 (Math|Boolean|ANDBoolean (Class|FortAthenaAIBotController|IsLocationInsideLeash _asfort_athena_aibot_controller (Transformation|GetActorLocation _output_get) true 0.0) (.ReturnValue (AI|Navigation|ProjectPointtoNavigation (Pawn|GetNavAgentLocation _output_get) (Fortnite|AI|GetNavigationDataforActor _output_get) "/Game/Athena/AI/Phoebe/NavFilter/NavFilter_Phoebe.NavFilter_Phoebe_C" (Math|Vector|MakeVector (ScalableFloat|GetValueatLevel (Variables|Default|GetTargetNavProbe_ExtentX)) (ScalableFloat|GetValueatLevel (Variables|Default|GetTargetNavProbe_ExtentY)) (ScalableFloat|GetValueatLevel (Variables|Default|GetTargetNavProbe_ExtentZ))))) (not (GameplayTags|HasAnyMatchingGameplayTags (CallFunction|GetAvatarasFortPlayerPawn) (GameplayTags|MakeGameplayTagContainerfromArray (Utilities|Array|MakeArray "(TagName=\"Gameplay.Status.Knockback\")")))) (not (AI|Components|NavMovement|IsFalling (Variables|/Script/Engine.Character|GetCharacterMovement (CallFunction|GetAvatarasFortPlayerPawn)))))))

(fn SelectDestinationCenter
  (bind _targetnavprobe_extentz (Variables|Default|GetTargetNavProbe_ExtentZ))
  (bind _returnvalue (ScalableFloat|GetValueatLevel _targetnavprobe_extentz))
  (bind _aipawn (Variables|Default|GetAIPawn))
  (bind _returnvalue_2 (Fortnite|AI|GetNavigationDataforActor _aipawn))
  (bind _targetnavprobe_extentx (Variables|Default|GetTargetNavProbe_ExtentX))
  (bind _returnvalue_3 (ScalableFloat|GetValueatLevel _targetnavprobe_extentx))
  (bind _targetnavprobe_extenty (Variables|Default|GetTargetNavProbe_ExtentY))
  (bind _returnvalue_4 (ScalableFloat|GetValueatLevel _targetnavprobe_extenty))
  (bind _targetactor (Class|BPFLAICommonFunctions|GetTargetActorfromBlackboard (Variables|Default|GetAIPawn)))
  (Utilities|IsValid _targetactor
    (:Is%20Valid
      (if (.ReturnValue (AI|Navigation|ProjectPointtoNavigation (Pawn|GetNavAgentLocation _targetactor) _returnvalue_2 "/Game/Athena/AI/Phoebe/NavFilter/NavFilter_Phoebe.NavFilter_Phoebe_C" (Math|Vector|MakeVector _returnvalue_3 _returnvalue_4 _returnvalue)))
        (return (AI|Navigation|ProjectPointtoNavigation (Pawn|GetNavAgentLocation _targetactor) _returnvalue_2 "/Game/Athena/AI/Phoebe/NavFilter/NavFilter_Phoebe.NavFilter_Phoebe_C" (Math|Vector|MakeVector _returnvalue_3 _returnvalue_4 _returnvalue)))
        (else
          _)))
    (:Is%20Not%20Valid
      _))
  (bind _haslastknownlocation (Class|BPPlayerPawnNonParticipantVividRazorCoffeePuff|HasValidLastKnownTargetLocation (Variables|Default|GetAIPawn)))
  (if _haslastknownlocation
    (return (AI|Components|Blackboard|GetValueasVector (AI|GetBlackboard (Variables|Default|GetAIBotController)) (Utilities|Name|MakeLiteralName "CoffeePuff_TargetLastKnownLocation")))
    (else
      (return (AI|Components|Blackboard|GetValueasVector (AI|GetBlackboard (Variables|Default|GetAIBotController)) (Utilities|Name|MakeLiteralName "CoffeePuff_SpawnLocation"))))))

(fn SelectTeleportPoint
  (params (EnvQueryResults :type EQS%20Query%20Instance%20Object%20Reference))
  (bind _resultlocations (AI|EQS|GetQueryResultsasLocations EnvQueryResults))
  (Variables|Default|SetTeleportTargetLocation (Utilities|Array|RandomArrayItem _resultlocations))
  (if (Variables|Default|GetAddEQSLocationsToKnown)
    (Class|BPPlayerPawnNonParticipantVividRazorCoffeePuff|AddListOfKnownValidLocations (Variables|Default|GetAIPawn) _resultlocations)))

(fn SpawnVisualization
  (bind _returnvalue (Building|SpawnBuildingGameplayActor "/VividRazor_CoffeePuff/GameplayActors/BGA_Blink_Visualization.BGA_Blink_Visualization_C" (Math|Transform|MakeTransform (+ (Variables|Default|GetTeleportTargetLocation) (Math|Vector|MakeVector 0.0 0.0 (Components|Capsule|GetScaledCapsuleHalfHeight (Variables|/Script/Engine.Character|GetCapsuleComponent (Variables|Default|GetAIPawn)))))) (Variables|Default|GetAIPawn) true false))
  (bind _asbga_blink_visualization (Utilities|Casting|CastToBGA_Blink_Visualization _returnvalue))
  (bind _output_get (Variables|Default|SetVisualizationActor _asbga_blink_visualization))
  (Class|BGABlinkVisualization|SetTargetPawn _output_get (Variables|Default|GetTargetPawn) :!class /VividRazor_CoffeePuff/GameplayActors/BGA_Blink_Visualization.BGA_Blink_Visualization_C))

(fn TryGloating
  (if (> (Utilities|Array|Length (Variables|Default|GetNewTargets)) 0)
    (bind _returnvalue (GameplayTags|HasMatchingGameplayTag(Message) (Variables|Default|GetAIPawn) "(TagName=\"Athena.AI.SpeechBubbles.VividRazor.CoffeePuff.MidCombatSpeechCooldown\")"))
    (if (not _returnvalue)
      (if (< (Math|Random|RandomFloat) 0.5)
        (Class|BPFLAICommonFunctions|PlaySpeechBubble (Variables|Default|GetAIPawn) (Variables|Default|GetGloatSpeechTag) true true)
        (Ability|ApplyGameplayEffectToOwner (Variables|Default|GetGloatCooldownTag))))))
