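; Gameplay ability for the CoffeePuff AI, apparently a teleport-punch variant for following its target, that adds an activation gate.
; CanActivateAbility reads the target actor from the avatar's AI blackboard.
; It passes only if the parent check succeeds, the target is moving on ground, and their vertical distance meets or exceeds a scalable threshold.
; Activate and end events are disabled stubs, leaving execution to the parent class.
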(event Ability|EventActivateAbility
  (|CallParentFunction) ; This node is disabled and will not be called.\nDrag off pins to build functionality.
)

(event Ability|EventOnEndAbility
  (params (bWasCancelled :type Boolean))
  (|CallParentFunction bWasCancelled) ; This node is disabled and will not be called.\nDrag off pins to build functionality.
)

(fn CanActivateAbility
  (params (ActorInfo :type Gameplay%20Ability%20Actor%20Info%20Structure) (Handle :type Gameplay%20Ability%20Spec%20Handle%20Structure))
  (bind _returnvalue (CallFunction|GetAvatarasFortPlayerPawn))
  (bind _targetactor (Class|BPFLAICommonFunctions|GetTargetActorfromBlackboard _returnvalue))
  (bind (_relevanttags _returnvalue_2) (|CallParentFunction ActorInfo Handle))
  (return _relevanttags (Math|Boolean|ANDBoolean _returnvalue_2 (AI|Components|NavMovement|IsMovingonGround (Variables|/Script/Engine.Character|GetCharacterMovement _targetactor)) (>= (Transformation|GetVerticalDistanceTo _returnvalue _targetactor) (ScalableFloat|GetValueatLevel (Variables|Default|GetFollowTargetVerticalDistanceThreshold))))))
